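# Physics asset for the BrightDisk medium female player body, set up for secondary motion on its jacket flaps.
# Capsules on head, upper spine, right clavicle and upper arms form body volumes, joined by free-angle constraints.
# Box-shaped dyn jacket bones chain from small base anchor bodies with mostly tight swing and twist limits.
# Upper-left and lower-right bases lock angularly to spine_05, while the upper-right base has no spine constraint.
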PhysicsAsset /BRCosmetics/Characters/Player/Female/Medium/Bodies/F_MED_BrightDisk/Meshes/F_MED_BrightDisk_Common_phys
bodies {
 base_down_jacket_r [{shapeType=Box, extentX=1.01, extentY=1.01, extentZ=1.01, shapeName=base_down_jacket_r_box, rotation=<rot 0 -90 0>}]
 base_up_jacket_l [{shapeType=Capsule, length=0.505, radius=0.505, shapeName=base_up_jacket_l_capsule, rotation=<rot 0 -90 0>}]
 base_up_jacket_r [{shapeType=Capsule, length=0.505, radius=0.505, shapeName=base_up_jacket_r_capsule, center=<0.5 0.5 0.5>}]
 clavicle_r [{shapeType=Capsule, length=4.8082, radius=5.5586, shapeName=clavicle_r_capsule, center=<-8.2 3.9285 0>, rotation=<rot 90 0 0>}]
 dyn_down_jacket_r [{
   shapeType=Box
   extentX=9.0112
   extentY=2.374
   extentZ=12.637
   shapeName=dyn_down_jacket_r_box
   center=<0.2145 3.9242 0.198>
   rotation=<rot -1.1592 -148.9256 -87.9318>
  }]
 dyn_up_jacket_01_l [{shapeType=Box, extentX=5.8001, extentY=1, extentZ=12.4508, rotation=<rot -24.2926 -0.3016 74.353>}]
 dyn_up_jacket_02_l [{
   shapeType=Box
   extentX=12.3631
   extentY=1.5653
   extentZ=17.9095
   shapeName=dyn_up_jacket_02_l_box
   center=<-0.3491 0.2409 -0.9086>
   rotation=<rot 31.0321 178.7348 -57.8215>
  }]
 dyn_up_jacket_02_r [{
   shapeType=Box
   extentX=9.4643
   extentY=15.527
   extentZ=1.2429
   shapeName=dyn_up_jacket_02_r_box
   center=<0.0251 0.1068 -0.9001>
   rotation=<rot 0.9421 -5.4614 -6.8867>
  }]
 dyn_up_jacket_03_r [{
   shapeType=Box
   extentX=5.782
   extentY=11.4041
   extentZ=0.8163
   shapeName=dyn_up_jacket_03_r_box
   center=<1.7929 0.7563 -0.9281>
   rotation=<rot 0.0001 4.8769 -14.4912>
  }]
 head [{
   shapeType=Capsule
   length=3.9662
   radius=5.9806
   shapeName=head_capsule
   center=<1.5529 0.8298 -0.1795>
   rotation=<rot 10.1097 39.9336 97.0481>
  }]
 spine_05 [{shapeType=Capsule, length=10.9, radius=11.3645, shapeName=spine_05_capsule, center=<-6.2214 2.3959 0>}]
 upperarm_l [{shapeType=Capsule, length=9.1044, radius=5.9378, center=<8.2066 -0.7454 0.3599>, rotation=<rot -90 0 180>}]
 upperarm_r [{shapeType=Capsule, length=14.4138, radius=5.9378, center=<-7.5433 0.5455 0>, rotation=<rot -90 178.8065 173.8008>}]
}
constraints [
 {bone1Name=dyn_up_jacket_03_r, bone2Name=dyn_up_jacket_02_r, swing1LimitDegrees=4, swing2LimitDegrees=64.72, twistLimitDegrees=13.52}
 {bone1Name=dyn_up_jacket_02_r, bone2Name=base_up_jacket_r, swing1LimitDegrees=2, swing2LimitDegrees=15, twistLimitDegrees=2}
 {
  bone1Name=base_up_jacket_l
  bone2Name=spine_05
  swing1LimitDegrees=45
  swing1Motion=Locked
  swing2LimitDegrees=45
  swing2Motion=Locked
  twistLimitDegrees=45
  twistMotion=Locked
 }
 {bone1Name=dyn_up_jacket_02_l, bone2Name=dyn_up_jacket_01_l, swing1LimitDegrees=5, swing2LimitDegrees=5, twistLimitDegrees=25}
 {
  bone1Name=upperarm_l
  bone2Name=spine_05
  swing1LimitDegrees=45
  swing1Motion=Free
  swing2LimitDegrees=45
  swing2Motion=Free
  twistLimitDegrees=45
  twistMotion=Free
 }
 {
  bone1Name=dyn_up_jacket_01_l
  bone2Name=base_up_jacket_l
  swing1LimitDegrees=45
  swing1Motion=Locked
  swing2LimitDegrees=45
  swing2Motion=Locked
  twistLimitDegrees=45
  twistMotion=Locked
 }
 {
  bone1Name=base_down_jacket_r
  bone2Name=spine_05
  swing1LimitDegrees=45
  swing1Motion=Locked
  swing2LimitDegrees=45
  swing2Motion=Locked
  twistLimitDegrees=45
  twistMotion=Locked
 }
 {bone1Name=dyn_down_jacket_r, bone2Name=base_down_jacket_r, swing1LimitDegrees=5, swing2LimitDegrees=7, twistLimitDegrees=1}
 {
  bone1Name=head
  bone2Name=spine_05
  swing1LimitDegrees=45
  swing1Motion=Free
  swing2LimitDegrees=45
  swing2Motion=Free
  twistLimitDegrees=45
  twistMotion=Free
 }
 {
  bone1Name=upperarm_r
  bone2Name=clavicle_r
  swing1LimitDegrees=45
  swing1Motion=Free
  swing2LimitDegrees=45
  swing2Motion=Free
  twistLimitDegrees=45
  twistMotion=Free
 }
 {
  bone1Name=clavicle_r
  bone2Name=spine_05
  swing1LimitDegrees=45
  swing1Motion=Free
  swing2LimitDegrees=45
  swing2Motion=Free
  twistLimitDegrees=45
  twistMotion=Free
 }
]
