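# Corrective Control Rig for the F_MED_BrightDisk female body, emote variant, fixing jacket, collar, belt, pants-zip and sneaker bones.
# After a shared DeformRig function, an "is emote" branch selects clavicle-constrained jacket fixes, foot-driven sneaker offsets and belt correctives.
# Weighted transform offsets and parent constraints take weights from remapped bone rotations, a thigh animation attribute and spherical pose readers.
# Further chains define arm-pose correctives for parachute and skydive, some Kalman-smoothed, plus DBNO offsets.
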ControlRigBlueprint /BRCosmetics/Characters/Player/Female/Medium/Bodies/F_MED_BrightDisk/Meshes/F_MED_BrightDisk_CtrlRig_Emote
graphs [
 RigVMFunctionLibrary {links [], nodes []}
 RigVMFunction_Sequence_1_1_ContainedGraph {nodes [
   n1 = RigVMFunctionEntryNode()
   n2 = &RigVMFunction_Sequence(ExecuteContext=@n1.ExecuteContext)
   n3 = &RigVMFunction_Sequence(ExecuteContext=@n2.B)
   n4 = &RigVMFunction_Sequence(ExecuteContext=@n3.B)
   n5 = &RigVMFunction_Sequence(ExecuteContext=@n4.B)
   n6 = &RigVMFunction_Sequence(ExecuteContext=@n5.B)
   n7 = &RigVMFunction_Sequence(ExecuteContext=@n6.B)
   RigVMFunctionReturnNode(A=@n2.A, B=@n3.A, C=@n4.A, D=@n5.A, E=@n6.A, F=@n7.A, G=@n7.B)
  ]}
 RigVMModel {nodes [
   Get Delta Ref_1 = RigVMFunctionReferenceNode(ExecuteContext=(), Item=(Type=Bone,Name="head"))
   Remap = &RigVMFunction_MathDoubleRemap(SourceMaximum=-30, SourceMinimum=-5, TargetMaximum=1, TargetMinimum=0, Value=@"Get Delta Ref_1".Result.Rotation.Yaw)
   > &RigUnit_ModifyTransforms(
    ItemToModify=((Item=(Type=Bone,Name="base_Collar_back_jacket"),Transform=(Rotation=(X=0.1964,Y=0.3773,Z=0.8028,W=0.4178),Translation=(X=0,Y=0,Z=0),Scale3D=(X=1,Y=1,Z=1))))
    WeightMinimum=0
    Weight=@Remap.Result
   )
   > &RigUnit_ModifyTransforms(
    ItemToModify="((Item=(Type=Bone,Name=\"base_side_jacket_r\"),Transform=(Rotation=(X=-0.0574,Y=-0.0486,Z=-0.096,W=0.9925),Translation=(X=0,Y=0,Z=0),Scale3D=(X=1,Y=1,Z=1))),(Item=(Type=Bone,Name=\"base_back_low_jacket_r\"),Transform=(Rotation=(X=0,Y=0,Z=-0.0922,W=0.9957),Tran ...(truncated)"
    WeightMinimum=0
    Weight=&RigVMFunction_MathDoubleRemap(SourceMaximum=2, SourceMinimum=-4, TargetMaximum=1, TargetMinimum=0, Value=@"Get Delta Ref_1".Result.Rotation.Yaw).Result
   )
   > &RigUnit_ModifyTransforms(
    ItemToModify="((Item=(Type=Bone,Name=\"base_side_jacket_l\"),Transform=(Rotation=(X=-0.0625,Y=0.0399,Z=0.0526,W=0.9959),Translation=(X=0,Y=0.7042,Z=0),Scale3D=(X=1,Y=1,Z=1))),(Item=(Type=Bone,Name=\"base_back_low_jacket_l\"),Transform=(Rotation=(X=0,Y=0,Z=0,W=1),Translation ...(truncated)"
    Weight=1
    WeightMinimum=0
   )
   Get Delta Ref_2_3 = RigVMFunctionReferenceNode(ExecuteContext=(), Item=(Type=Bone,Name="foot_l"))
   > &RigUnit_ModifyTransforms(
    ItemToModify=((Item=(Type=Bone,Name="sneakerBack_L"),Transform=(Rotation=(X=0,Y=0.0175,Z=0,W=0.9998),Translation=(X=3,Y=0,Z=1),Scale3D=(X=1,Y=1,Z=1))))
    WeightMinimum=0
    Weight=&RigVMFunction_MathDoubleRemap(SourceMaximum=50, SourceMinimum=30, TargetMaximum=1, TargetMinimum=0, Value=@"Get Delta Ref_2_3".Result.Rotation.Yaw).Result
   )
   > Get Delta Ref_2_3_1 = RigVMFunctionReferenceNode(Item=(Type=Bone,Name="foot_r"))
   > &RigUnit_ModifyTransforms(
    ItemToModify=((Item=(Type=Bone,Name="sneakerBack_R"),Transform=(Rotation=(X=0,Y=0,Z=0,W=1),Translation=(X=-2,Y=0,Z=1),Scale3D=(X=1,Y=1,Z=1))))
    WeightMinimum=0
    Weight=&RigVMFunction_MathDoubleRemap(SourceMaximum=60, SourceMinimum=25, TargetMaximum=1, TargetMinimum=0, Value=@"Get Delta Ref_2_3_1".Result.Rotation.Yaw).Result
   )
   n1 = &RigVMFunction_MathDoubleMul(B=1, A=RigVMVariableNode() {variable=Parashut spherical}.Value)
   &RigUnit_ModifyTransforms(
    ExecutePin=()
    ItemToModify="((Item=(Type=Bone,Name=\"dyn_up_jacket_02_r\"),Transform=(Rotation=(X=0.2092,Y=-0.2758,Z=0.3407,W=0.8741),Translation=(X=0,Y=0,Z=0),Scale3D=(X=1,Y=1,Z=1))),(Item=(Type=Bone,Name=\"dyn_up_jacket_04_r\"),Transform=(Rotation=(X=0.0858,Y=-0.173,Z=0.0151,W=0.9811), ...(truncated)"
    WeightMinimum=0
    Weight=@n1.Result
   )
   > &RigUnit_ModifyTransforms(
    ItemToModify=((Item=(Type=Bone,Name="base_up_jacket_l"),Transform=(Rotation=(X=-0.1166,Y=-0.2651,Z=-0.3853,W=0.8762),Translation=(X=1,Y=0,Z=0),Scale3D=(X=1,Y=1,Z=1))))
    WeightMinimum=0
    Weight=@n1.Result
   )
   > &RigUnit_ModifyTransforms(
    ItemToModify="((Item=(Type=Bone,Name=\"foot_r\"),Transform=(Rotation=(X=-0.0385,Y=0.0061,Z=-0.1556,W=0.987),Translation=(X=0,Y=3,Z=-1.612),Scale3D=(X=1,Y=1,Z=1))),(Item=(Type=Bone,Name=\"foot_l\"),Transform=(Rotation=(X=0,Y=0,Z=-0.1373,W=0.9905),Translation=(X=0,Y=-3,Z=-1), ...(truncated)"
    WeightMinimum=0
    Weight=@n1.Result
   )
   > &RigUnit_ModifyTransforms(
    ItemToModify="((Item=(Type=Bone,Name=\"None\"),Transform=(Rotation=(X=0,Y=0,Z=0,W=1),Translation=(X=0,Y=0,Z=0),Scale3D=(X=1,Y=1,Z=1))),(Item=(Type=Bone,Name=\"dyn_down_jacket_01_r\"),Transform=(Rotation=(X=0,Y=0,Z=0,W=1),Translation=(X=0,Y=0,Z=0),Scale3D=(X=1,Y=1,Z=1))),(It ...(truncated)"
    WeightMinimum=0
    Weight=@n1.Result
   )
   > &RigUnit_ModifyTransforms(
    ItemToModify=((Item=(Type=Bone,Name="master_up_jacket_r"),Transform=(Rotation=(X=-0.0216,Y=0.4833,Z=-0.0391,W=0.8743),Translation=(X=0,Y=0,Z=0),Scale3D=(X=1,Y=1,Z=1))))
    WeightMinimum=0
    Weight=@n1.Result
   )
   > &RigUnit_ModifyTransforms(
    ItemToModify="((Item=(Type=Bone,Name=\"sneakerBack_L\"),Transform=(Rotation=(X=0,Y=-0.1253,Z=0,W=0.9921),Translation=(X=1.5001,Y=0,Z=1.1283),Scale3D=(X=1,Y=1,Z=1))),(Item=(Type=Bone,Name=\"sneakerBack_R\"),Transform=(Rotation=(X=0,Y=-0.1767,Z=0,W=0.9843),Translation=(X=0,Y= ...(truncated)"
    Weight=1
    WeightMinimum=0
   )
   > &RigUnit_ModifyTransforms(
    ItemToModify="((Item=(Type=Bone,Name=\"None\"),Transform=(Rotation=(X=0,Y=0,Z=0,W=1),Translation=(X=0,Y=0,Z=0),Scale3D=(X=1,Y=1,Z=1))),(Item=(Type=Bone,Name=\"base_Collar_back_jacket\"),Transform=(Rotation=(X=-0.0159,Y=0.0768,Z=0.3131,W=0.9465),Translation=(X=0.8979,Y=0,Z=0 ...(truncated)"
    WeightMinimum=0
    Weight=RigVMVariableNode() {variable=Skydive collar correcting}.Value
   )
   n2 = &RigVMFunction_KalmanFloat(BufferSize=5, Value=RigVMVariableNode() {variable=Right spherical}.Value) {template="KalmanFilter::Execute(in Value,in BufferSize,out Result)"}
   &RigUnit_ModifyTransforms(
    ExecutePin=()
    ItemToModify=((Item=(Type=Bone,Name="master_up_jacket_r"),Transform=(Rotation=(X=0.0389,Y=0.5808,Z=0.0543,W=0.8113),Translation=(X=0,Y=0,Z=0),Scale3D=(X=1,Y=1,Z=1))))
    Mode=OverrideLocal
    WeightMinimum=0
    Weight=@n2.Result
   )
   > &RigUnit_ModifyTransforms(
    ItemToModify="((Item=(Type=Bone,Name=\"dyn_up_jacket_02_r\"),Transform=(Rotation=(X=-0.27,Y=0.1539,Z=0.339,W=0.888),Translation=(X=-0.3308,Y=-2.137,Z=-2.9535),Scale3D=(X=0.8804,Y=1,Z=1))),(Item=(Type=Bone,Name=\"dyn_up_jacket_03_r\"),Transform=(Rotation=(X=-0.0843,Y=-0.0145 ...(truncated)"
    Mode=OverrideLocal
    WeightMinimum=0
    Weight=@n2.Result
   )
   > &RigUnit_ModifyTransforms(
    ItemToModify="((Item=(Type=Bone,Name=\"dyn_down_jacket_01_r\"),Transform=(Rotation=(X=0.0855,Y=-0.5494,Z=0.0566,W=0.8293),Translation=(X=-0.1903,Y=0.9919,Z=0.0806),Scale3D=(X=1,Y=1,Z=1))),(Item=(Type=Bone,Name=\"dyn_down_jacket_02_r\"),Transform=(Rotation=(X=0.0223,Y=-0.241 ...(truncated)"
    WeightMinimum=0
    Weight=@n2.Result
   )
   > &RigUnit_ModifyTransforms(
    ItemToModify="((Item=(Type=Bone,Name=\"base_up_jacket_l\"),Transform=(Rotation=(X=0.0104,Y=-0.2732,Z=-0.0364,W=0.9612),Translation=(X=0,Y=0,Z=0),Scale3D=(X=1,Y=1,Z=1))),(Item=(Type=Bone,Name=\"dyn_up_jacket_02_l\"),Transform=(Rotation=(X=-0.0587,Y=-0.0171,Z=-0.2395,W=0.969) ...(truncated)"
    WeightMinimum=0
    Weight=@n2.Result
   )
   > &RigUnit_ModifyTransforms(
    ItemToModify=((Item=(Type=Bone,Name="base_Collar_back_jacket"),Transform=(Rotation=(X=0,Y=0,Z=0.197,W=0.9804),Translation=(X=0,Y=0,Z=0),Scale3D=(X=1,Y=1,Z=1))))
    WeightMinimum=0
    Weight=@n2.Result
   )
   n3 = RigVMVariableNode() {variable=Right spherical}
   &RigUnit_ModifyTransforms(
    ExecutePin=()
    ItemToModify="((Item=(Type=Bone,Name=\"None\"),Transform=(Rotation=(X=0,Y=0,Z=0,W=1),Translation=(X=0,Y=0,Z=0),Scale3D=(X=1,Y=1,Z=1))),(Item=(Type=Bone,Name=\"dyn_up_jacket_02_r\"),Transform=(Rotation=(X=-0.504,Y=0.574,Z=0.3188,W=0.5611),Translation=(X=-1.0189,Y=-0.5736,Z=0 ...(truncated)"
    WeightMinimum=0
    Weight=@n3.Value
   )
   > &RigUnit_ModifyTransforms(
    ItemToModify="((Item=(Type=Bone,Name=\"base_up_jacket_l\"),Transform=(Rotation=(X=0.0104,Y=-0.2732,Z=-0.0364,W=0.9612),Translation=(X=0,Y=0,Z=0),Scale3D=(X=1,Y=1,Z=1))),(Item=(Type=Bone,Name=\"dyn_up_jacket_02_l\"),Transform=(Rotation=(X=-0.0848,Y=-0.0649,Z=-0.2488,W=0.9626 ...(truncated)"
    WeightMinimum=0
    Weight=@n3.Value
   )
   DISPATCH_RigDispatch_GetAnimAttribute = RigVMDispatchNode(BoneName=root, Default=0, Name=deformRig_l_thigh_up_alpha) {template="DISPATCH_RigDispatch_GetAnimAttribute(in Name,in BoneName,in Default,out Value,out Found)"}
   &RigUnit_ParentConstraint(
    AdvancedSettings=(InterpolationType=Average,RotationOrderForFilter=XZY)
    Child=(Type=Bone,Name="base_belt_m")
    ExecutePin=()
    Parents=((Item=(Type=Bone,Name="pelvis"),Weight=0.2),(Item=(Type=Bone,Name="base_belt_r"),Weight=0.4),(Item=(Type=Bone,Name="base_belt_l"),Weight=0.4))
    Weight=@DISPATCH_RigDispatch_GetAnimAttribute.Value
   )
   > &RigUnit_ModifyTransforms(
    ItemToModify="((Item=(Type=Bone,Name=\"base_belt_l\"),Transform=(Rotation=(X=0.0851,Y=-0.0472,Z=0.0472,W=0.9941),Translation=(X=-10.0973,Y=2.1521,Z=-5.9811),Scale3D=(X=1,Y=1,Z=1))),(Item=(Type=Bone,Name=\"base_belt_m\"),Transform=(Rotation=(X=0.05,Y=-0.0111,Z=0.2875,W=0.956 ...(truncated)"
    WeightMinimum=0
    Weight=@DISPATCH_RigDispatch_GetAnimAttribute.Value
   )
   > &RigUnit_ModifyTransforms(
    ItemToModify=((Item=(Type=Bone,Name="base_fix_l"),Transform=(Rotation=(X=0,Y=0,Z=0.1736,W=0.9848),Translation=(X=-7.8358,Y=-1.6444,Z=-3),Scale3D=(X=1,Y=1,Z=1.5))))
    WeightMinimum=0
    Weight=@DISPATCH_RigDispatch_GetAnimAttribute.Value
   )
   > &RigUnit_ModifyTransforms(
    ItemToModify="((Item=(Type=Bone,Name=\"base_belt_r\"),Transform=(Rotation=(X=0.1285,Y=0.0227,Z=0.1722,W=0.9764),Translation=(X=-1.758,Y=2.528,Z=0.3966),Scale3D=(X=1.2834,Y=1.2834,Z=0.8014))),(Item=(Type=Bone,Name=\"base_last_bottle\"),Transform=(Rotation=(X=0,Y=0,Z=0,W=1),T ...(truncated)"
    WeightMinimum=0
    Weight=@DISPATCH_RigDispatch_GetAnimAttribute.Value
   )
   > &RigUnit_ModifyTransforms(
    ItemToModify="((Item=(Type=Bone,Name=\"None\"),Transform=(Rotation=(X=0,Y=0,Z=0,W=1),Translation=(X=0,Y=0,Z=0),Scale3D=(X=1,Y=1,Z=1))),(Item=(Type=Bone,Name=\"base_rt_cor\"),Transform=(Rotation=(X=0.1616,Y=0.2536,Z=0.5124,W=0.8044),Translation=(X=-11.154,Y=1.3231,Z=2.8323), ...(truncated)"
    WeightMinimum=0
    Weight=@DISPATCH_RigDispatch_GetAnimAttribute.Value
   )
   n4 = &RigVMFunction_MathDoubleRemap(
    SourceMaximum=-15
    SourceMinimum=-10
    TargetMaximum=1
    TargetMinimum=0
    Value=&RigVMFunction_MathQuaternionToEuler(
     RotationOrder=XYZ
     Value=&RigUnit_GetRelativeTransformForItem(Child=(Type=Bone,Name="spine_04"), Parent=(Type=Bone,Name="spine_03"), bChildInitial=False, bParentInitial=False).RelativeTransform.Rotation
    ).Result.Z
   )
   &RigUnit_ParentConstraint(
    AdvancedSettings=(InterpolationType=Average,RotationOrderForFilter=XZY)
    Child=(Type=Bone,Name="base_back_low_jacket")
    ExecutePin=()
    Parents=((Item=(Type=Bone,Name="spine_02"),Weight=0.5),(Item=(Type=Bone,Name="spine_01"),Weight=0.5))
    Weight=@n4.Result
   )
   > &RigUnit_ParentConstraint(
    AdvancedSettings=(InterpolationType=Average,RotationOrderForFilter=XZY)
    Child=(Type=Bone,Name="dyn_back_low_jacket_l_1")
    Parents=((Item=(Type=Bone,Name="spine_02"),Weight=0.5),(Item=(Type=Bone,Name="spine_01"),Weight=0.5))
    Weight=@n4.Result
   )
   > &RigUnit_ParentConstraint(
    AdvancedSettings=(InterpolationType=Average,RotationOrderForFilter=XZY)
    Child=(Type=Bone,Name="dyn_back_low_jacket_r_1")
    Parents=((Item=(Type=Bone,Name="spine_02"),Weight=0.5),(Item=(Type=Bone,Name="spine_01"),Weight=0.5))
    Weight=@n4.Result
   )
   RigUnit_BeginExecution = &RigUnit_BeginExecution()
   FN_DeformRig_FNC = RigVMFunctionReferenceNode(
    DeformRig_ Curve="(deformRig_l_upperarm_up_alpha_72_#1=False,deformRig_r_upperarm_up_alpha_73_#2=false,deformRig_l_upperarm_fr_alpha_74_#3=False,deformRig_r_upperarm_fr_alpha_75_CB51B8 ...(truncated)"
    DeformRig_Bool="(upperarm_1_#4=True,elbow_3_#5=True,wrist_7_#6=True,groin_8_#7=True,glute_17_#8=True,knee_9_40C408324CA261 ...(truncated)"
    DeformRig_Debug=(upperarm_debug_18_#4=False,wrist_debug_20_#6=False,lat_debug_26_#9=False,groinGlute_debug_27_#7=False)
    DeformRig_Value="(upperarm_up_18_#10=(Rotation=(X=0,Y=0,Z=0,W=1),Translation=(X=0,Y=0,Z=9),Scale3D=(X=0.7,Y=1,Z=1)),upperarm_fwd_fr_19_#11=(Rotation=(X=0,Y=0,Z=0,W=1),Translation=(X=0,Y=14,Z=0),Scale3D=(X=0.5,Y=1,Z= ...(truncated)"
    ExecuteContext=@RigUnit_BeginExecution.ExecutePin
   )
   > n5 = &RigUnit_ParentConstraint(
    AdvancedSettings=(InterpolationType=Average,RotationOrderForFilter=XZY)
    Child=(Type=Bone,Name="pants_zip_correct")
    Parents=((Item=(Type=Bone,Name="None"),Weight=1),(Item=(Type=Bone,Name="pelvis"),Weight=0.2),(Item=(Type=Bone,Name="thigh_l"),Weight=0.4),(Item=(Type=Bone,Name="thigh_r"),Weight=0.4))
    Weight=1
   )
   RigVMFunction_ControlFlowBranch = &RigVMFunction_ControlFlowBranch(ExecuteContext=@n5.ExecutePin, Condition=RigVMVariableNode() {variable=is emote}.Value)
   RigVMFunction_Sequence_1 = RigVMAggregateNode(ExecuteContext=@RigVMFunction_ControlFlowBranch.False)
   &RigUnit_ParentConstraint(
    AdvancedSettings=(InterpolationType=Average,RotationOrderForFilter=XZY)
    Child=(Type=Bone,Name="base_up_jacket_r")
    Filter=(TranslationFilter=(bX=true,bY=true,bZ=true),RotationFilter=(bX=false,bY=false,bZ=false),ScaleFilter=(bX=True,bY=True,bZ=True))
    Parents=((Item=(Type=Bone,Name="clavicle_r"),Weight=1))
    Weight=1
    ExecutePin=@RigVMFunction_ControlFlowBranch.True
   )
   > &RigUnit_ParentConstraint(
    AdvancedSettings=(InterpolationType=Average,RotationOrderForFilter=XZY)
    Child=(Type=Bone,Name="parent_up_jacket_l")
    Filter=(TranslationFilter=(bX=false,bY=false,bZ=false),RotationFilter=(bX=true,bY=true,bZ=true),ScaleFilter=(bX=True,bY=True,bZ=True))
    Parents=((Item=(Type=Bone,Name="clavicle_l"),Weight=1))
    Weight=1
   )
   > &RigUnit_ModifyTransforms(
    ItemToModify=((Item=(Type=Bone,Name="base_Collar_back_jacket"),Transform=(Rotation=(X=0,Y=0,Z=0.2588,W=0.9659),Translation=(X=0,Y=0,Z=0),Scale3D=(X=1,Y=1,Z=1))))
    Weight=1
    WeightMinimum=0
   )
   > &RigUnit_ModifyTransforms(
    ItemToModify="((Item=(Type=Bone,Name=\"deform_groin_base_r\"),Transform=(Rotation=(X=0,Y=0,Z=0,W=1),Translation=(X=-1,Y=0,Z=-1),Scale3D=(X=1,Y=1,Z=1))),(Item=(Type=Bone,Name=\"cor_rt_01\"),Transform=(Rotation=(X=0,Y=0,Z=0.2588,W=0.9659),Translation=(X=-3,Y=0,Z=-2.739),Scale ...(truncated)"
    Weight=1
    WeightMinimum=0
   )
   > Get Delta Ref_2_3_2 = RigVMFunctionReferenceNode(Item=(Type=Bone,Name="foot_l"))
   > &RigUnit_ModifyTransforms(
    ItemToModify=((Item=(Type=Bone,Name="sneakerBack_L"),Transform=(Rotation=(X=0,Y=0.0175,Z=0,W=0.9998),Translation=(X=3,Y=0,Z=1),Scale3D=(X=1,Y=1,Z=1))))
    WeightMinimum=0
    Weight=&RigVMFunction_MathDoubleRemap(SourceMaximum=50, SourceMinimum=30, TargetMaximum=1, TargetMinimum=0, Value=@"Get Delta Ref_2_3_2".Result.Rotation.Yaw).Result
   )
   > Get Delta Ref_2_3_1_1 = RigVMFunctionReferenceNode(Item=(Type=Bone,Name="foot_r"))
   > &RigUnit_ModifyTransforms(
    ItemToModify="((Item=(Type=Bone,Name=\"sneakerBack_R\"),Transform=(Rotation=(X=0,Y=0,Z=0,W=1),Translation=(X=-2,Y=0,Z=1),Scale3D=(X=1,Y=1,Z=1))),(Item=(Type=Bone,Name=\"base_belt_m\"),Transform=(Rotation=(X=0,Y=0,Z=0,W=1),Translation=(X=0,Y=0,Z=-1),Scale3D=(X=1,Y=1,Z=1))),( ...(truncated)"
    WeightMinimum=0
    Weight=&RigVMFunction_MathDoubleRemap(SourceMaximum=60, SourceMinimum=25, TargetMaximum=1, TargetMinimum=0, Value=@"Get Delta Ref_2_3_1_1".Result.Rotation.Yaw).Result
   )
   n6 = RigVMDispatchNode(BoneName=root, Default=0, Name=deformRig_l_thigh_up_alpha) {template="DISPATCH_RigDispatch_GetAnimAttribute(in Name,in BoneName,in Default,out Value,out Found)"}
   > &RigUnit_ParentConstraint(
    AdvancedSettings=(InterpolationType=Average,RotationOrderForFilter=XZY)
    Child=(Type=Bone,Name="base_belt_m")
    Parents=((Item=(Type=Bone,Name="pelvis"),Weight=0.2),(Item=(Type=Bone,Name="base_belt_r"),Weight=0.4),(Item=(Type=Bone,Name="base_belt_l"),Weight=0.4))
    Weight=@n6.Value
   )
   > &RigUnit_ModifyTransforms(
    ItemToModify="((Item=(Type=Bone,Name=\"base_belt_r\"),Transform=(Rotation=(X=0.1285,Y=0.0227,Z=0.1722,W=0.9764),Translation=(X=-1.758,Y=2.528,Z=0.3966),Scale3D=(X=1.2834,Y=1.2834,Z=0.8014))),(Item=(Type=Bone,Name=\"base_last_bottle\"),Transform=(Rotation=(X=0,Y=0,Z=0,W=1),T ...(truncated)"
    WeightMinimum=0
    Weight=@n6.Value
   )
   > &RigUnit_ModifyTransforms(
    ItemToModify="((Item=(Type=Bone,Name=\"None\"),Transform=(Rotation=(X=0,Y=0,Z=0,W=1),Translation=(X=0,Y=0,Z=0),Scale3D=(X=1,Y=1,Z=1))),(Item=(Type=Bone,Name=\"base_rt_cor\"),Transform=(Rotation=(X=0.1616,Y=0.2536,Z=0.5124,W=0.8044),Translation=(X=-11.154,Y=1.3231,Z=2.8323), ...(truncated)"
    WeightMinimum=0
    Weight=@n6.Value
   )
   > &RigUnit_ModifyTransforms(
    ItemToModify=((Item=(Type=Bone,Name="base_fix_l"),Transform=(Rotation=(X=0,Y=0,Z=0.1736,W=0.9848),Translation=(X=-7.8358,Y=-1.6444,Z=-3),Scale3D=(X=1,Y=1,Z=1.5))))
    WeightMinimum=0
    Weight=@n6.Value
   )
   > &RigUnit_ModifyTransforms(
    ItemToModify="((Item=(Type=Bone,Name=\"base_belt_l\"),Transform=(Rotation=(X=0.0851,Y=-0.0472,Z=0.0472,W=0.9941),Translation=(X=-10.0973,Y=2.1521,Z=-5.9811),Scale3D=(X=1,Y=1,Z=1))),(Item=(Type=Bone,Name=\"base_belt_m\"),Transform=(Rotation=(X=0.05,Y=-0.0111,Z=0.2875,W=0.956 ...(truncated)"
    WeightMinimum=0
    Weight=@n6.Value
   )
   > n7 = &RigUnit_SphericalPoseReader(
    ActiveRegionScaleFactors=(PositiveWidth=0.3,NegativeWidth=0.3,PositiveHeight=1,NegativeHeight=1)
    ActiveRegionSize=0.1
    Debug=(bDrawDebug=True,bDraw2D=False,bDrawLocalAxes=False,DebugScale=25,DebugSegments=20,DebugThickness=0.25)
    DriverAxis=(X=-1,Y=0,Z=0)
    DriverItem=(Type=Bone,Name="thigh_l")
    FalloffRegionScaleFactors=(PositiveWidth=0.1,NegativeWidth=0.2,PositiveHeight=0.1,NegativeHeight=0.1)
    FalloffSize=0.2
    FlipHeightScaling=False
    FlipWidthScaling=False
    OptionalParentItem=(Type=Bone,Name="None")
    RotationOffset=(X=166,Y=10,Z=-5)
   )
   > &RigUnit_ModifyTransforms(
    ItemToModify="((Item=(Type=Bone,Name=\"None\"),Transform=(Rotation=(X=0,Y=0,Z=0,W=1),Translation=(X=0,Y=0,Z=0),Scale3D=(X=1,Y=1,Z=1))),(Item=(Type=Bone,Name=\"base_rt_cor\"),Transform=(Rotation=(X=0.0174,Y=-0.1655,Z=0.1401,W=0.9761),Translation=(X=-13.43,Y=4.4402,Z=2.3301), ...(truncated)"
    WeightMinimum=0
    Weight=@n7.OutputParam
   )
   > &RigUnit_ModifyTransforms(
    ItemToModify=((Item=(Type=Bone,Name="base_fix_l"),Transform=(Rotation=(X=0,Y=0,Z=0.1939,W=0.981),Translation=(X=-4.6205,Y=0.0813,Z=-0.045),Scale3D=(X=0.8272,Y=1.1281,Z=1.5104))))
    WeightMinimum=0
    Weight=@n7.OutputParam
   )
   > &RigUnit_ModifyTransforms(
    ItemToModify="((Item=(Type=Bone,Name=\"base_belt_l\"),Transform=(Rotation=(X=0.0851,Y=-0.0472,Z=0.0472,W=0.9941),Translation=(X=-0.9954,Y=2.0893,Z=0.0215),Scale3D=(X=1,Y=1,Z=1))),(Item=(Type=Bone,Name=\"base_belt_m\"),Transform=(Rotation=(X=0.05,Y=-0.0111,Z=0.2875,W=0.9564) ...(truncated)"
    WeightMinimum=0
    Weight=@n7.OutputParam
   )
   n8 = &RigVMFunction_ControlFlowBranch(ExecuteContext=(), Condition=RigVMVariableNode() {variable=Is DBNO}.Value)
   &RigUnit_ModifyTransforms(
    ItemToModify=((Item=(Type=Bone,Name="base_side_jacket_r"),Transform=(Rotation=(X=0.1145,Y=-0.0997,Z=0.0276,W=0.988),Translation=(X=0.3364,Y=-4.2234,Z=-0.9592),Scale3D=(X=1,Y=1,Z=1))))
    Weight=1
    WeightMinimum=0
    ExecutePin=@n8.True
   )
   > &RigUnit_ModifyTransforms(
    ItemToModify="((Item=(Type=Bone,Name=\"base_side_jacket_l\"),Transform=(Rotation=(X=-0.0405,Y=-0.0505,Z=-0.0634,W=0.9959),Translation=(X=0.1611,Y=2.591,Z=0.1066),Scale3D=(X=1,Y=1,Z=1))),(Item=(Type=Bone,Name=\"base_back_low_jacket_r\"),Transform=(Rotation=(X=0,Y=0,Z=0,W=1), ...(truncated)"
    Weight=1
    WeightMinimum=0
   )
   n9 = &RigUnit_SphericalPoseReader(
    ActiveRegionScaleFactors=(PositiveWidth=0.45,NegativeWidth=0.4,PositiveHeight=0.4,NegativeHeight=0.5)
    ActiveRegionSize=0.3
    Debug=(bDrawDebug=true,bDraw2D=false,bDrawLocalAxes=False,DebugScale=25,DebugSegments=20,DebugThickness=0.25)
    DriverAxis=(X=-1,Y=0,Z=0)
    DriverItem=(Type=Bone,Name="upperarm_r")
    ExecutePin=()
    FalloffRegionScaleFactors=(PositiveWidth=0.4,NegativeWidth=0.3,PositiveHeight=0.25,NegativeHeight=0.3)
    FalloffSize=0.5
    FlipHeightScaling=False
    FlipWidthScaling=false
    OptionalParentItem=(Type=None,Name="spine_05")
    RotationOffset=(X=-50.8879,Y=139.8867,Z=58.5913)
   )
   > RigVMVariableNode(Value=@n9.OutputParam) {variable=Right spherical}
   > n10 = &RigUnit_SphericalPoseReader(
    ActiveRegionScaleFactors=(PositiveWidth=1,NegativeWidth=1,PositiveHeight=1,NegativeHeight=1)
    ActiveRegionSize=0.1
    Debug=(bDrawDebug=false,bDraw2D=False,bDrawLocalAxes=False,DebugScale=25,DebugSegments=20,DebugThickness=0.25)
    DriverAxis=(X=1,Y=0,Z=0)
    DriverItem=(Type=Bone,Name="upperarm_l")
    FalloffRegionScaleFactors=(PositiveWidth=3,NegativeWidth=0.2,PositiveHeight=1,NegativeHeight=1)
    FalloffSize=0.2
    FlipHeightScaling=False
    FlipWidthScaling=False
    OptionalParentItem=(Type=Bone,Name="None")
    RotationOffset=(X=-5.1666,Y=-50,Z=0)
   )
   > RigVMVariableNode(Value=@n10.OutputParam) {variable=Parashut spherical}
   > n11 = &RigUnit_SphericalPoseReader(
    ActiveRegionScaleFactors=(PositiveWidth=1,NegativeWidth=1,PositiveHeight=1,NegativeHeight=1)
    ActiveRegionSize=0.22
    Debug=(bDrawDebug=True,bDraw2D=False,bDrawLocalAxes=False,DebugScale=25,DebugSegments=20,DebugThickness=0.25)
    DriverAxis=(X=1,Y=0,Z=0)
    DriverItem=(Type=Bone,Name="upperarm_l")
    FalloffRegionScaleFactors=(PositiveWidth=3,NegativeWidth=1,PositiveHeight=1,NegativeHeight=1)
    FalloffSize=0.3
    FlipHeightScaling=False
    FlipWidthScaling=False
    OptionalParentItem=(Type=Bone,Name="None")
    RotationOffset=(X=-8.0543,Y=-80.859,Z=7.5728)
   )
   > RigVMVariableNode(Value=@n11.OutputParam) {variable=Skydive collar correcting}
   DISPATCH_RigVMDispatch_Print = RigVMDispatchNode(Enabled=True, ExecuteContext=(), ScreenColor=(R=1,G=1,B=1,A=1), ScreenDuration=0.05, Value=0) {template="DISPATCH_RigVMDispatch_Print(in Prefix,in Value,in Enabled,in ScreenDuration,in ScreenColor)"}
   &RigUnit_ModifyTransforms(
    ExecutePin=()
    ItemToModify="((Item=(Type=Bone,Name=\"base_up_jacket_l\"),Transform=(Rotation=(X=-0.1166,Y=-0.2651,Z=-0.3853,W=0.8762),Translation=(X=1,Y=0,Z=0),Scale3D=(X=1,Y=1,Z=1))),(Item=(Type=Bone,Name=\"base_Collar_back_jacket\"),Transform=(Rotation=(X=0.1341,Y=0.2198,Z=0.3664,W=0.8 ...(truncated)"
    Weight=1
    WeightMinimum=0
   )
   &RigUnit_ModifyTransforms(
    ExecutePin=()
    ItemToModify=((Item=(Type=Bone,Name="deform_upperarm_fr_r"),Transform=(Rotation=(X=0,Y=0,Z=-0.0087,W=1),Translation=(X=0,Y=-1,Z=0),Scale3D=(X=1.1,Y=1.1,Z=1.1))))
    WeightMinimum=0
    Weight=@n2.Result
   )
   &RigUnit_ModifyTransforms(
    ExecutePin=()
    ItemToModify=((Item=(Type=Bone,Name="base_fix_l"),Transform=(Rotation=(X=0,Y=0,Z=0.1939,W=0.981),Translation=(X=-1.4723,Y=-1.2315,Z=-0.045),Scale3D=(X=0.8272,Y=1.1281,Z=1.5104))))
    Weight=1
    WeightMinimum=0
   )
   &RigUnit_ModifyTransforms(
    ExecutePin=()
    ItemToModify=((Item=(Type=Bone,Name="base_Collar_back_jacket"),Transform=(Rotation=(X=-0.0984,Y=0.1705,Z=0.4902,W=0.8491),Translation=(X=0,Y=0,Z=0),Scale3D=(X=1,Y=1,Z=1))))
    WeightMinimum=0
    Weight=RigVMVariableNode() {variable=Skydive collar correcting}.Value
   )
   &RigUnit_ModifyTransforms(
    ExecutePin=()
    ItemToModify="((Item=(Type=Bone,Name=\"dyn_up_jacket_02_r\"),Transform=(Rotation=(X=-0.2965,Y=0.1284,Z=-0.0721,W=-0.9436),Translation=(X=-1.0927,Y=-0.5654,Z=-5.6838),Scale3D=(X=0,Y=0,Z=0))),(Item=(Type=Bone,Name=\"dyn_up_jacket_01_r\"),Transform=(Rotation=(X=0.0418,Y=-0.008 ...(truncated)"
    Weight=1
    WeightMinimum=0
   )
   &RigUnit_GetRelativeTransformForItem(Child=(Type=Bone,Name="clavicle_r"), Parent=(Type=Bone,Name="clavicle_r"), bChildInitial=true, bParentInitial=false)
   &RigUnit_SphericalPoseReader(
    ActiveRegionScaleFactors=(PositiveWidth=1,NegativeWidth=1,PositiveHeight=1,NegativeHeight=1)
    ActiveRegionSize=0.1
    Debug=(bDrawDebug=true,bDraw2D=False,bDrawLocalAxes=False,DebugScale=25,DebugSegments=20,DebugThickness=0.25)
    DriverAxis=(X=1,Y=0,Z=0)
    DriverItem=(Type=Bone,Name="upperarm_l")
    ExecutePin=()
    FalloffRegionScaleFactors=(PositiveWidth=1,NegativeWidth=1,PositiveHeight=1,NegativeHeight=1)
    FalloffSize=0.4
    FlipHeightScaling=False
    FlipWidthScaling=false
    OptionalParentItem=(Type=Bone,Name="None")
    RotationOffset=(X=65,Y=0,Z=-40)
   )
   &RigUnit_SphericalPoseReader(
    ActiveRegionScaleFactors=(PositiveWidth=0.45,NegativeWidth=0.3,PositiveHeight=0.04,NegativeHeight=0.5)
    ActiveRegionSize=0.4
    Debug=(bDrawDebug=false,bDraw2D=false,bDrawLocalAxes=False,DebugScale=25,DebugSegments=20,DebugThickness=0.25)
    DriverAxis=(X=-1,Y=0,Z=0)
    DriverItem=(Type=Bone,Name="upperarm_r")
    ExecutePin=()
    FalloffRegionScaleFactors=(PositiveWidth=0.14,NegativeWidth=0.3,PositiveHeight=0.25,NegativeHeight=0.3)
    FalloffSize=0.5
    FlipHeightScaling=False
    FlipWidthScaling=False
    OptionalParentItem=(Type=Bone,Name="spine_05")
    RotationOffset=(X=-62.025,Y=139.8867,Z=58.5913)
   )
   &RigUnit_SphericalPoseReader(
    ActiveRegionScaleFactors=(PositiveWidth=0.45,NegativeWidth=0.3,PositiveHeight=0.04,NegativeHeight=0.5)
    ActiveRegionSize=0.4
    Debug=(bDrawDebug=true,bDraw2D=false,bDrawLocalAxes=False,DebugScale=25,DebugSegments=20,DebugThickness=0.25)
    DriverAxis=(X=-1,Y=0,Z=0)
    DriverItem=(Type=Bone,Name="upperarm_r")
    ExecutePin=()
    FalloffRegionScaleFactors=(PositiveWidth=1,NegativeWidth=0.3,PositiveHeight=0.25,NegativeHeight=0.3)
    FalloffSize=0.5
    FlipHeightScaling=False
    FlipWidthScaling=False
    OptionalParentItem=(Type=Bone,Name="spine_05")
    RotationOffset=(X=-62.025,Y=139.8867,Z=58.5913)
   )
   &RigUnit_SphericalPoseReader(
    ActiveRegionScaleFactors=(PositiveWidth=1,NegativeWidth=1,PositiveHeight=1,NegativeHeight=1)
    ActiveRegionSize=0.1
    Debug=(bDrawDebug=true,DebugScale=25,DebugSegments=20,DebugThickness=0.25)
    DriverAxis=(X=-1,Y=0,Z=0)
    DriverItem=(Type=Bone,Name="upperarm_r")
    ExecutePin=()
    FalloffRegionScaleFactors=(PositiveWidth=1,NegativeWidth=1,PositiveHeight=1,NegativeHeight=1)
    FalloffSize=0.5
    FlipHeightScaling=true
    FlipWidthScaling=false
    OptionalParentItem=(Type=Bone,Name="None")
    RotationOffset=(X=180,Y=0,Z=180)
   )
  ]}
]
hierarchy [
 root {loc=<0 0 0>, rot=<0 0 0 1>}
 pelvis {parent=root, loc=<0 -0.5301 98.0571>, rot=<0.0127 0.707 -0.0127 -0.707>}
 spine_01 {parent=pelvis, loc=<6.1208 -0.9328 0>, rot=<0 0 0.034 -0.9994>}
 spine_02 {parent=spine_01, loc=<6.5996 -0.2095 0>, rot=<0 0 0.0238 -0.9997>}
 spine_03 {parent=spine_02, loc=<6.4914 -0.1237 0>, rot=<0 0 0.0149 -0.9999>}
 spine_04 {parent=spine_03, loc=<6.9598 0.0794 0>, rot=<0 0 -0.0132 -0.9999>}
 spine_05 {parent=spine_04, loc=<6.9987 0 0>, rot=<0 0 -0.0015 -1>}
 clavicle_l {parent=spine_05, loc=<5.4831 -0.1053 -3.2946>, rot=<0.0465 -0.7289 -0.009 -0.6829>}
 upperarm_l {parent=clavicle_l, loc=<9.8658 -1.2136 -0.4105>, rot=<0.0164 -0.3245 -0.0269 -0.9454>}
 lowerarm_l {parent=upperarm_l, loc=<26.465 -0.8345 0.259>, rot=<0.0036 0.0161 -0.2298 -0.9731>}
 hand_l {parent=lowerarm_l, loc=<22.7355 -0.3203 -0.6105>, rot=<-0.6784 -0.0223 0.0735 -0.7307>}
 thumb_01_l {parent=hand_l, loc=<1.3118 -0.3191 -2.459>, rot=<0.6546 -0.4042 0.0664 -0.6354>}
 thumb_02_l {parent=thumb_01_l, loc=<5.9336 -0.0448 0.0401>, rot=<-0.0037 0.0213 0.298 -0.9543>}
 thumb_03_l {parent=thumb_02_l, loc=<2.2223 0.3682 -0.0907>, rot=<0.0118 -0.013 0.0066 -0.9998>}
 index_metacarpal_l {parent=hand_l, loc=<1.6535 0.0183 -2.96>, rot=<0.0158 -0.0124 -0.0173 -0.9996>}
 index_01_l {parent=index_metacarpal_l, loc=<8.3008 0.0357 -0.0524>, rot=<-0.0044 0.0485 0.0924 -0.9945>}
 index_02_l {parent=index_01_l, loc=<3.3587 -0.3004 0.0452>, rot=<-0.002 0.0027 0.3225 -0.9466>}
 index_03_l {parent=index_02_l, loc=<2.5578 -0.123 -0.0252>, rot=<-0.002 0.0067 0.2558 -0.9667>}
 middle_metacarpal_l {parent=hand_l, loc=<2.1856 -0.1725 -0.5984>, rot=<0.006 0.0168 -0.0329 -0.9993>}
 middle_01_l {parent=middle_metacarpal_l, loc=<7.3275 -0.2447 0.0129>, rot=<-0.0208 0.0042 0.0858 -0.9961>}
 middle_02_l {parent=middle_01_l, loc=<4.0123 0.1401 0.1175>, rot=<-0.0218 0.0258 0.3512 -0.9357>}
 middle_03_l {parent=middle_02_l, loc=<2.6077 0.0534 0.1373>, rot=<0.0082 0.0188 0.1708 -0.9851>}
 ring_metacarpal_l {parent=hand_l, loc=<2.2074 -0.3625 1.3638>, rot=<-0.0284 0.0722 -0.0458 -0.9959>}
 ring_01_l {parent=ring_metacarpal_l, loc=<7.0295 -0.6197 -0.1446>, rot=<-0.0489 -0.0252 0.1892 -0.9804>}
 ring_02_l {parent=ring_01_l, loc=<3.6174 0.4647 -0.1027>, rot=<0 0 0.2806 -0.9598>}
 ring_03_l {parent=ring_02_l, loc=<2.4152 0.114 0.0199>, rot=<0 0 0.1189 -0.9929>}
 pinky_metacarpal_l {parent=hand_l, loc=<1.7616 -0.9753 3.3814>, rot=<-0.0397 0.0814 -0.069 -0.9935>}
 pinky_01_l {parent=pinky_metacarpal_l, loc=<7.1709 -0.4831 -0.1094>, rot=<-0.0654 -0.0035 0.2242 -0.9723>}
 pinky_02_l {parent=pinky_01_l, loc=<2.8274 0.0893 -0.1043>, rot=<0.0037 0.0011 0.2854 -0.9584>}
 pinky_03_l {parent=pinky_02_l, loc=<2.2046 0.0307 -0.0078>, rot=<-0.0002 0.0098 0.1445 -0.9895>}
 deform_wrist_dn_base_l {parent=hand_l, loc=<0 -1.5326 0>, rot=<0 0 0 -1>}
 deform_wrist_dn_l {parent=deform_wrist_dn_base_l, loc=<0 0 0>, rot=<0 0 0 1>}
 deform_wrist_up_base_l {parent=hand_l, loc=<0 2.3009 0>, rot=<0 0 0 -1>}
 deform_wrist_up_l {parent=deform_wrist_up_base_l, loc=<0 0 0>, rot=<0 0 0 1>}
 lowerarm_twist_02_l {parent=lowerarm_l, loc=<10.6707 0 0>, rot=<0 0 0 1>}
 lowerarm_twist_01_l {parent=lowerarm_l, loc=<18.316 0 0>, rot=<0 0 0 1>}
 deform_elbow_in_base_l {parent=lowerarm_l, loc=<0.1834 3.6062 -0.2572>, rot=<0 0 0 -1>}
 deform_elbow_in_l {parent=deform_elbow_in_base_l, loc=<0 0 0>, rot=<0 0 0 1>}
 deform_elbow_out_base_l {parent=lowerarm_l, loc=<0 -4 0>, rot=<0 0 0 1>}
 deform_elbow_out_l {parent=deform_elbow_out_base_l, loc=<0 0 0>, rot=<0 0 0 1>}
 upperarm_twist_01_l {parent=upperarm_l, loc=<7.2135 -0.1134 0>, rot=<0 0 0 1>}
 deform_upperarm_bk_base_l {parent=upperarm_twist_01_l, loc=<-6.4053 -4.3746 0.314>, rot=<0 0 0 -1>}
 deform_upperarm_bk_l {parent=deform_upperarm_bk_base_l, loc=<0 0 0>, rot=<0 0 0 1>}
 deform_upperarm_fr_base_l {parent=upperarm_twist_01_l, loc=<-5.6312 5.092 0.3354>, rot=<0 0 0 -1>}
 deform_upperarm_fr_l {parent=deform_upperarm_fr_base_l, loc=<0 0 0>, rot=<0 0 0 1>}
 deform_upperarm_up_base_l {parent=upperarm_twist_01_l, loc=<-8.891 0.4825 3.9517>, rot=<0 0 0 -1>}
 deform_upperarm_up_l {parent=deform_upperarm_up_base_l, loc=<0 0 0>, rot=<0 0 0 1>}
 upperarm_twist_02_l {parent=upperarm_l, loc=<19.7998 -0.1134 0>, rot=<0 0 0 1>}
 neck_01 {parent=spine_05, loc=<11.3138 0.1529 0>, rot=<0 0 -0.2003 -0.9797>}
 neck_02 {parent=neck_01, loc=<4.5627 0 0>, rot=<0 0 0.0504 -0.9987>}
 head {parent=neck_02, loc=<5.3914 0 0>, rot=<0 0 0.115 -0.9934>}
 clavicle_r {parent=spine_05, loc=<5.4831 -0.1053 3.2946>, rot=<0.7289 0.0465 -0.6829 0.009>}
 upperarm_r {parent=clavicle_r, loc=<-9.8658 1.2136 0.4104>, rot=<0.0164 -0.3245 -0.0269 -0.9454>}
 lowerarm_r {parent=upperarm_r, loc=<-26.4651 0.8345 -0.259>, rot=<0.0036 0.0161 -0.2298 -0.9731>}
 hand_r {parent=lowerarm_r, loc=<-22.7356 0.3203 0.6106>, rot=<-0.6784 -0.0223 0.0735 -0.7307>}
 thumb_01_r {parent=hand_r, loc=<-1.3119 0.3192 2.459>, rot=<0.6546 -0.4042 0.0664 -0.6354>}
 thumb_02_r {parent=thumb_01_r, loc=<-5.9335 0.0447 -0.0401>, rot=<-0.0037 0.0213 0.298 -0.9543>}
 thumb_03_r {parent=thumb_02_r, loc=<-2.2223 -0.3682 0.0907>, rot=<0.0118 -0.013 0.0066 -0.9998>}
 index_metacarpal_r {parent=hand_r, loc=<-1.6534 -0.0181 2.9601>, rot=<0.0158 -0.0124 -0.0173 -0.9996>}
 index_01_r {parent=index_metacarpal_r, loc=<-8.3008 -0.0357 0.0524>, rot=<-0.0044 0.0485 0.0924 -0.9945>}
 index_02_r {parent=index_01_r, loc=<-3.3589 0.3003 -0.0453>, rot=<-0.002 0.0027 0.3225 -0.9466>}
 index_03_r {parent=index_02_r, loc=<-2.5578 0.123 0.0252>, rot=<-0.002 0.0067 0.2558 -0.9667>}
 middle_metacarpal_r {parent=hand_r, loc=<-2.1856 0.1726 0.5984>, rot=<0.006 0.0168 -0.0329 -0.9993>}
 middle_01_r {parent=middle_metacarpal_r, loc=<-7.3273 0.2447 -0.0129>, rot=<-0.0208 0.0042 0.0858 -0.9961>}
 middle_02_r {parent=middle_01_r, loc=<-4.0124 -0.14 -0.1175>, rot=<-0.0218 0.0258 0.3512 -0.9357>}
 middle_03_r {parent=middle_02_r, loc=<-2.6077 -0.0533 -0.1373>, rot=<0.0082 0.0188 0.1708 -0.9851>}
 ring_metacarpal_r {parent=hand_r, loc=<-2.2073 0.3627 -1.3638>, rot=<-0.0284 0.0722 -0.0458 -0.9959>}
 ring_01_r {parent=ring_metacarpal_r, loc=<-7.0297 0.6197 0.1446>, rot=<-0.0489 -0.0252 0.1892 -0.9804>}
 ring_02_r {parent=ring_01_r, loc=<-3.6171 -0.4648 0.1027>, rot=<0 0 0.2806 -0.9598>}
 ring_03_r {parent=ring_02_r, loc=<-2.4153 -0.1141 -0.0199>, rot=<0 0 0.1189 -0.9929>}
 pinky_metacarpal_r {parent=hand_r, loc=<-1.7617 0.9753 -3.3814>, rot=<-0.0397 0.0814 -0.069 -0.9935>}
 pinky_01_r {parent=pinky_metacarpal_r, loc=<-7.1707 0.4832 0.1094>, rot=<-0.0654 -0.0035 0.2242 -0.9723>}
 pinky_02_r {parent=pinky_01_r, loc=<-2.8274 -0.0892 0.1043>, rot=<0.0037 0.0011 0.2854 -0.9584>}
 pinky_03_r {parent=pinky_02_r, loc=<-2.2045 -0.031 0.0078>, rot=<-0.0002 0.0098 0.1445 -0.9895>}
 deform_wrist_dn_base_r {parent=hand_r, loc=<-0.0001 1.5328 0>, rot=<0 0 0 -1>}
 deform_wrist_dn_r {parent=deform_wrist_dn_base_r, loc=<0 0 0>, rot=<0 0 0 1>}
 deform_wrist_up_base_r {parent=hand_r, loc=<-0.0002 -2.3007 -0.0001>, rot=<0 0 0 -1>}
 deform_wrist_up_r {parent=deform_wrist_up_base_r, loc=<0 0 0>, rot=<0 0 0 1>}
 lowerarm_twist_02_r {parent=lowerarm_r, loc=<-10.6707 0 0>, rot=<0 0 0 1>}
 lowerarm_twist_01_r {parent=lowerarm_r, loc=<-18.3161 0 0>, rot=<0 0 0 1>}
 deform_elbow_in_base_r {parent=lowerarm_r, loc=<-0.1834 -3.6063 0.2572>, rot=<0 0 0 -1>}
 deform_elbow_in_r {parent=deform_elbow_in_base_r, loc=<0 0 0>, rot=<0 0 0 1>}
 deform_elbow_out_base_r {parent=lowerarm_r, loc=<0 4 -0.0001>, rot=<0 0 0 1>}
 deform_elbow_out_r {parent=deform_elbow_out_base_r, loc=<0 0 0>, rot=<0 0 0 1>}
 upperarm_twist_01_r {parent=upperarm_r, loc=<-7.2135 0.1134 0>, rot=<0 0 0 1>}
 deform_upperarm_bk_base_r {parent=upperarm_twist_01_r, loc=<6.4052 4.3746 -0.314>, rot=<0 0 0 -1>}
 deform_upperarm_bk_r {parent=deform_upperarm_bk_base_r, loc=<0 0 0>, rot=<0 0 0 1>}
 deform_upperarm_fr_base_r {parent=upperarm_twist_01_r, loc=<5.6311 -5.092 -0.3354>, rot=<0 0 0 -1>}
 deform_upperarm_fr_r {parent=deform_upperarm_fr_base_r, loc=<0 0 0>, rot=<0 0 0 1>}
 deform_upperarm_up_base_r {parent=upperarm_twist_01_r, loc=<8.891 -0.4825 -3.9517>, rot=<0 0 0 -1>}
 deform_upperarm_up_r {parent=deform_upperarm_up_base_r, loc=<0 0 0>, rot=<0 0 0 1>}
 upperarm_twist_02_r {parent=upperarm_r, loc=<-19.7998 0.1134 0>, rot=<0 0 0 1>}
 attach_cape {parent=spine_05, loc=<9.0493 -11.7296 -0.0014>, rot=<0.0615 -0.7044 -0.0616 -0.7044>}
 attach_backpack {parent=spine_05, loc=<-0.0736 -13.4888 -0.0019>, rot=<-0.0094 -0.707 0.0093 -0.707>}
 base_Collar_back_jacket {parent=spine_05, loc=<13.9631 -3.255 0.8223>, rot=<0.9995 0.0273 0.0179 -0.0049>}
 dyn_Collar_back_jacket {parent=base_Collar_back_jacket, loc=<0 0 0>, rot=<0 0 0 1>}
 base_up_jacket_r {parent=spine_05, loc=<12.2066 0.0874 6.0927>, rot=<-0.2091 0.4076 -0.0923 -0.8841>}
 dyn_up_jacket_01_r {parent=base_up_jacket_r, loc=<0 0 0>, rot=<0 0 0 1>}
 dyn_up_jacket_02_r {parent=dyn_up_jacket_01_r, loc=<4.082 1.1989 0.2418>, rot=<-0.0875 -0.0755 0.052 -0.9919>}
 dyn_up_jacket_03_r {parent=dyn_up_jacket_02_r, loc=<4.0699 0.1456 -0.182>, rot=<0 0 0.0974 -0.9952>}
 parent_up_jacket_l {parent=spine_05, loc=<9.6676 2.6515 -5.9274>, rot=<0.0284 -0.7065 -0.0284 -0.7065>}
 base_up_jacket_l {parent=parent_up_jacket_l, loc=<0 0 0>, rot=<0.6102 0.3573 0.1405 -0.693>}
 dyn_up_jacket_01_l {parent=base_up_jacket_l, loc=<0 0 0>, rot=<0 0 0 1>}
 dyn_up_jacket_02_l {parent=dyn_up_jacket_01_l, loc=<3.5981 0 0>, rot=<0 0 0 1>}
 Base_low_jacket_l {parent=spine_05, loc=<5.7791 7.0955 -6.733>, rot=<-0.0222 -0.0471 0.9764 0.2095>}
 dyn_low_jacket_l {parent=Base_low_jacket_l, loc=<0 0 0>, rot=<0 0 0 1>}
 collar_l_semi_corr {parent=spine_05, loc=<12.8705 -0.7482 -5.5633>, rot=<0.0284 -0.7065 -0.0284 -0.7065>}
 base_down_jacket_r {parent=spine_05, loc=<4.8862 7.2569 6.6592>, rot=<-0.2878 0.5477 -0.3509 -0.7029>}
 dyn_down_jacket_r {parent=base_down_jacket_r, loc=<0 0 0>, rot=<0 0 0 1>}
 base_side_jacket_l {parent=spine_04, loc=<-0.2065 11.7558 -8.5357>, rot=<0.0464 -0.9377 -0.3442 -0.0058>}
 dyn_side_jacket_01_l {parent=base_side_jacket_l, loc=<0 0 0>, rot=<0 0 0 1>}
 base_side_jacket_r {parent=spine_04, loc=<-0.2067 11.5784 8.5357>, rot=<0.0058 0.3442 0.9377 -0.0464>}
 dyn_side_jacket_01_r {parent=base_side_jacket_r, loc=<0 0 0>, rot=<0 0 0 1>}
 base_zip {parent=dyn_side_jacket_01_r, loc=<11.4335 0.1706 -1.4762>, rot=<-0.407 -0.0737 -0.0095 -0.9104>}
 dyn_zip {parent=base_zip, loc=<0 0 0>, rot=<0 0 0 1>}
 base_back_low_jacket {parent=spine_04, loc=<-1.3472 -7.4733 -0.1759>, rot=<-0.0033 0.696 0.7181 -0.005>}
 dyn_back_low_jacket {parent=base_back_low_jacket, loc=<0 0 0>, rot=<0 0 0 1>}
 base_back_low_jacket_r {parent=spine_04, loc=<-1.3475 -5.4063 10.714>, rot=<-0.0033 0.696 0.7181 -0.005>}
 dyn_back_low_jacket_r_1 {parent=base_back_low_jacket_r, loc=<0 0 0>, rot=<0 0 0 1>}
 base_back_low_jacket_l {parent=spine_04, loc=<-1.4025 -4.7459 -10.4249>, rot=<-0.0033 0.696 0.7181 -0.005>}
 dyn_back_low_jacket_l_1 {parent=base_back_low_jacket_l, loc=<0 0 0>, rot=<0 0 0 1>}
 thigh_l {parent=pelvis, loc=<-1.0855 -0.2109 -8.7053>, rot=<0.0009 0.0355 0.0155 -0.9993>}
 calf_l {parent=thigh_l, loc=<-42.5284 -0.0019 -0.7297>, rot=<0.0006 -0.0161 -0.0391 -0.9991>}
 foot_l {parent=calf_l, loc=<-46.7402 -2.5014 -0.592>, rot=<0.0427 -0.0238 0.0353 -0.9982>}
 ball_l {parent=foot_l, loc=<-3.5658 10.8066 -0.1544>, rot=<0.0001 -0.0001 -0.7063 -0.7079>}
 calf_twist_02_l {parent=calf_l, loc=<-19.1546 -1.7943 -0.2638>, rot=<0.0237 0.0095 -0.0035 -0.9997>}
 base_boots_zip_l {parent=calf_twist_02_l, loc=<-2.546 0.851 4.816>, rot=<-0.04 -0.934 0.3543 0.0214>}
 dyn_boots_zip_l_1 {parent=base_boots_zip_l, loc=<0 0 0>, rot=<0 0 0 1>}
 calf_twist_01_l {parent=calf_l, loc=<-35.8483 -1.9197 -0.5794>, rot=<0.0237 0.0095 -0.0035 -0.9997>}
 sneakerBack_L {parent=calf_twist_01_l, loc=<-5.3054 -4.4516 1.1723>, rot=<0.479 -0.5492 0.4936 -0.4748>}
 deform_knee_in_base_l {parent=calf_l, loc=<0.8031 -5.0231 0.2456>, rot=<0 0 0 -1>}
 deform_knee_in_l {parent=deform_knee_in_base_l, loc=<0 0 0>, rot=<0 0 0 1>}
 deform_knee_out_base_l {parent=calf_l, loc=<0 4 0>, rot=<0 0 0 -1>}
 deform_knee_out_l {parent=deform_knee_out_base_l, loc=<0 0 0>, rot=<0 0 0 1>}
 thigh_twist_01_l {parent=thigh_l, loc=<-27.952 0 0>, rot=<0.0013 0 0 -1>}
 base_bottle_correct {parent=thigh_twist_01_l, loc=<22.1121 -0.4717 -11.4503>, rot=<-0.0518 -0.0884 0.9947 -0.002>}
 dyn_bottle_correct_1 {parent=base_bottle_correct, loc=<0 0 0>, rot=<0 0 0 1>}
 base_last_bottle {parent=thigh_twist_01_l, loc=<19.6278 -7.1039 -9.0919>, rot=<-0.0518 -0.0884 0.9947 -0.002>}
 dyn_last_bottle_1 {parent=base_last_bottle, loc=<0 0 0>, rot=<0 0 0 1>}
 deform_groin_base_l {parent=thigh_l, loc=<0.4801 5.6776 -0.2486>, rot=<0 0 0 -1>}
 deform_groin_l {parent=deform_groin_base_l, loc=<0 0 0>, rot=<0 0 0 1>}
 thigh_r {parent=pelvis, loc=<-1.0853 -0.2109 8.7053>, rot=<0.0355 -0.0009 0.9993 0.0155>}
 calf_r {parent=thigh_r, loc=<42.5286 0.0019 0.7297>, rot=<0.0006 -0.0161 -0.0391 -0.9991>}
 foot_r {parent=calf_r, loc=<46.7403 2.5014 0.592>, rot=<0.0427 -0.0238 0.0353 -0.9982>}
 ball_r {parent=foot_r, loc=<3.5658 -10.8066 0.1544>, rot=<0.0001 -0.0001 -0.7063 -0.7079>}
 calf_twist_02_r {parent=calf_r, loc=<19.1546 1.7943 0.2638>, rot=<0.0237 0.0095 -0.0035 -0.9997>}
 base_boots_zip_r {parent=calf_twist_02_r, loc=<2.546 -0.851 -4.816>, rot=<0.3269 -0.0603 0.0407 -0.9423>}
 dyn_boots_zip_r_1 {parent=base_boots_zip_r, loc=<0 0 0>, rot=<0 0 0 1>}
 calf_twist_01_r {parent=calf_r, loc=<35.8485 1.9197 0.5794>, rot=<0.0237 0.0095 -0.0035 -0.9997>}
 sneakerBack_R {parent=calf_twist_01_r, loc=<5.3053 4.4516 -1.1724>, rot=<0.479 -0.5492 0.4936 -0.4748>}
 deform_knee_in_base_r {parent=calf_r, loc=<-0.8032 5.0231 -0.2457>, rot=<0 0 0 -1>}
 deform_knee_in_r {parent=deform_knee_in_base_r, loc=<0 0 0>, rot=<0 0 0 1>}
 deform_knee_out_base_r {parent=calf_r, loc=<-0.0001 -4 0>, rot=<0 0 0 -1>}
 deform_knee_out_r {parent=deform_knee_out_base_r, loc=<0 0 0>, rot=<0 0 0 1>}
 thigh_twist_01_r {parent=thigh_r, loc=<27.952 0 0>, rot=<0.0013 0 0 -1>}
 deform_groin_base_r {parent=thigh_r, loc=<-0.48 -5.6776 0.2486>, rot=<0 0 0 -1>}
 deform_groin_r {parent=deform_groin_base_r, loc=<0 0 0>, rot=<0 0 0 1>}
 deform_glute_base_l {parent=pelvis, loc=<-7.3466 -9.0213 -9.1245>, rot=<0.0172 0.031 -0.4477 -0.8935>}
 deform_glute_l {parent=deform_glute_base_l, loc=<0 0 0>, rot=<0 0 0 1>}
 deform_glute_base_r {parent=pelvis, loc=<-7.3464 -9.0213 9.1245>, rot=<0.031 -0.0172 0.8935 -0.4477>}
 deform_glute_r {parent=deform_glute_base_r, loc=<0 0 0>, rot=<0 0 0 1>}
 base_belt_m {parent=pelvis, loc=<2.3415 8.3942 -0.2303>, rot=<-0.0202 -0.706 0.7079 0.0052>}
 cor_belt_m_1 {parent=base_belt_m, loc=<0 0 0>, rot=<0 0 0 1>}
 base_belt_r {parent=pelvis, loc=<5.3835 7.6937 6.054>, rot=<-0.0202 -0.706 0.7079 0.0052>}
 cor_belt_r_1 {parent=base_belt_r, loc=<0 0 0>, rot=<0 0 0 1>}
 pants_zip_correct {parent=pelvis, loc=<-0.0209 6.4107 0>, rot=<-0.5089 -0.491 -0.491 -0.5089>}
 base_fix_l {parent=pelvis, loc=<-0.3042 6.9826 -7.7596>, rot=<0.0609 -0.6836 0.7143 0.137>}
 cor_fix_1 {parent=base_fix_l, loc=<0 0 0>, rot=<0 0 0 1>}
 base_belt_l {parent=pelvis, loc=<-0.0574 7.6742 -7.8907>, rot=<-0.0202 -0.706 0.7079 0.0052>}
 cor_belt_l_1 {parent=base_belt_l, loc=<0 0 0>, rot=<0 0 0 1>}
 base_rt_cor {parent=pelvis, loc=<-2.464 6.946 -10.8271>, rot=<-0.0127 -0.707 0.707 0.0127>}
 cor_rt_01 {parent=base_rt_cor, loc=<0 0 0>, rot=<0 0 0 1>}
]
variables [
 Is DBNO {type=bool, default=False}
 Parashut spherical {type=double, default=0}
 Right spherical {type=double, default=0}
 Skydive collar correcting {type=double, default=0}
 is emote {type=bool, default=False}
]
