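# Post-process style AnimBP for the F_MED_BrightDisk female body, adding secondary motion to jacket and boot zips over the pose copied from the parent mesh.
# Control Rig feeds AnimDynamics and RigidBody, with separate emote and gameplay chains; wearing a cape adds a collider RigidBody pass.
# The event graph updates gravity, backbling flags and a head-rotation alpha for the right jacket bone.
# Several jacket AnimDynamics and bone-offset nodes are left unconnected.
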AnimBlueprint /BRCosmetics/Characters/Player/Female/Medium/Bodies/F_MED_BrightDisk/Meshes/F_MED_BrightDisk_AnimBP
target_skeleton=&/F_MED_BrightDisk_Skeleton
graphs [
 AnimGraph {nodes [
   "Anim Dynamics - Chain: base_up_jacket_l -> dyn_up_jacket_02_l"!(
    SimulationSpace:AnimPhysSimSpaceType Enum=BoneRelative
    ComponentPose
    Alpha=1
    GravityOverride:Vector=Get GravityOverride()
    ExternalForce="0,0,0"
   ) -> (Pose) {
    angularDampingOverride=1
    angularSpringConstant=500
    bAngularSpring=true
    bChain=true
    bGravityOverrideInSimSpace=true
    bOverrideAngularDamping=true
    bUseGravityOverride=true
    bUseSphericalLimits=true
    alphaBoolBlend {blendOption=Linear}
    boundBone {boneName=base_up_jacket_l}
    chainEnd {boneName=dyn_up_jacket_02_l}
    componentAppliedLinearAccClamp=<5000 5000 5000>
    componentLinearAccScale=<0.55 0.55 0.55>
    componentLinearVelScale=<0.5 0.5 0.5>
    physicsBodyDefinitions [
     {
      collisionType=OuterSphere
      boundBone {boneName=base_up_jacket_l}
      boxExtents=<6 6 6>
      constraintSetup {angularLimitsMax=<0 0 5>, angularLimitsMin=<-10 -10 -5>, angularTarget=<1 0 0>}
      localJointOffset=<-1 0 0>
     }
     {
      collisionType=OuterSphere
      boundBone {boneName=dyn_up_jacket_01_l}
      boxExtents=<6 6 6>
      constraintSetup {angularLimitsMax=<0 0 5>, angularLimitsMin=<-10 -10 -5>, angularTarget=<1 0 0>}
      localJointOffset=<-1 0 0>
     }
     {
      collisionType=OuterSphere
      boundBone {boneName=dyn_up_jacket_02_l}
      boxExtents=<6 6 6>
      constraintSetup {angularLimitsMax=<0 0 5>, angularLimitsMin=<-10 -10 -5>, angularTarget=<1 0 0>}
      localJointOffset=<-1 0 0>
     }
    ]
    relativeSpaceBone {boneName=spine_04}
    sphericalLimits [{limitRadius=7, drivingBone {boneName=head}, sphereLocalOffset=<-0.5713 3.9637 -9.3447>}]
   }
   n1 = Anim Dynamics - Bone: dyn_low_jacket_l(Alpha=1, ExternalForce="0,0,0") {
    angularDampingOverride=1
    angularSpringConstant=500
    bAngularSpring=true
    bGravityOverrideInSimSpace=true
    bOverrideAngularDamping=true
    bUseGravityOverride=true
    simulationSpace=BoneRelative
    alphaBoolBlend {blendOption=Linear}
    boundBone {boneName=dyn_low_jacket_l}
    componentAppliedLinearAccClamp=<5000 5000 5000>
    componentLinearAccScale=<0.55 0.55 0.55>
    componentLinearVelScale=<0.5 0.5 0.5>
    gravityOverride {z=-980}
    physicsBodyDefinitions [{
      collisionType=OuterSphere
      boundBone {boneName=dyn_low_jacket_l}
      boxExtents=<6 6 6>
      constraintSetup {angularLimitsMax=<0 0 5>, angularLimitsMin=<-10 -10 -5>, angularTarget=<1 0 0>}
      localJointOffset=<-1 0 0>
     }]
    relativeSpaceBone {boneName=clavicle_l}
   }
   Anim Dynamics - Bone: dyn_low_jacket_r(ComponentPose=@n1, Alpha=1, ExternalForce="0,0,0") {
    angularDampingOverride=1
    angularSpringConstant=500
    bAngularSpring=true
    bGravityOverrideInSimSpace=true
    bOverrideAngularDamping=true
    bUseGravityOverride=true
    simulationSpace=BoneRelative
    alphaBoolBlend {blendOption=Linear}
    boundBone {boneName=dyn_low_jacket_r}
    componentAppliedLinearAccClamp=<5000 5000 5000>
    componentLinearAccScale=<0.55 0.55 0.55>
    componentLinearVelScale=<0.5 0.5 0.5>
    gravityOverride {z=-980}
    physicsBodyDefinitions [{
      collisionType=OuterSphere
      boundBone {boneName=dyn_low_jacket_r}
      boxExtents=<6 6 6>
      constraintSetup {angularLimitsMax=<0 0 5>, angularLimitsMin=<-10 -10 -5>, angularTarget=<1 0 0>}
      localJointOffset=<-1 0 0>
     }]
    relativeSpaceBone {boneName=clavicle_r}
   }
   Anim Dynamics - Bone: dyn_up_jacket_02_r!(SimulationSpace:AnimPhysSimSpaceType Enum=Component, ComponentPose, Alpha=1, GravityOverride:Vector="0,0,0", ExternalForce="0,0,0") -> (Pose) {
    angularBiasOverride=0.95
    angularDampingOverride=0.95
    angularSpringConstant=160
    bAngularSpring=true
    bLinearSpring=true
    bOverrideAngularBias=true
    bOverrideAngularDamping=true
    bOverrideLinearDamping=true
    bUseGravityOverride=true
    bUseSphericalLimits=true
    linearDampingOverride=0.95
    linearSpringConstant=160
    simulationSpace=BoneRelative
    alphaBoolBlend {blendOption=Linear}
    boundBone {boneName=dyn_up_jacket_02_r}
    componentAppliedLinearAccClamp=<150 150 150>
    componentLinearAccScale=<0.1 0.75 0.75>
    componentLinearVelScale=<0.1 0.75 0.75>
    physicsBodyDefinitions [{
      boundBone {boneName=dyn_up_jacket_02_r}
      boxExtents=<16 5 10>
      constraintSetup {angularLimitsMax=<0 0 13>, angularLimitsMin=<0 0 -3>, angularTarget=<0 0 0>}
      localJointOffset=<1 0 0>
     }]
    relativeSpaceBone {boneName=spine_04}
    sphericalLimits [{limitRadius=7.4967, drivingBone {boneName=head}, sphereLocalOffset=<5 5 5>}]
   }
   n2 = Anim Dynamics - Bone: dyn_up_jacket_02_r!(ComponentPose, Alpha=1, GravityOverride=Get GravityOverride(), ExternalForce="0,0,0") -> (Pose) {
    angularDampingOverride=1
    angularSpringConstant=500
    bAngularSpring=true
    bGravityOverrideInSimSpace=true
    bOverrideAngularDamping=true
    bUseGravityOverride=true
    bUseSphericalLimits=true
    alphaBoolBlend {blendOption=Linear}
    boundBone {boneName=dyn_up_jacket_02_r}
    componentAppliedLinearAccClamp=<5000 5000 5000>
    componentLinearAccScale=<0.55 0.55 0.55>
    componentLinearVelScale=<0.5 0.5 0.5>
    gravityOverride {z=-980}
    physicsBodyDefinitions [{
      collisionType=OuterSphere
      boundBone {boneName=dyn_up_jacket_02_r}
      boxExtents=<6 6 6>
      constraintSetup {angularLimitsMax=<0 0 5>, angularLimitsMin=<-10 -10 -5>, angularTarget=<1 0 0>}
      localJointOffset=<-1 0 0>
     }]
    relativeSpaceBone {boneName=spine_04}
    sphericalLimits [{limitRadius=8, drivingBone {boneName=head}, sphereLocalOffset=<0 1.6607 8.0756>}]
   }
   Anim Dynamics - Bone: dyn_up_jacket_02_l!(ComponentPose=@n2, Alpha=1, GravityOverride:Vector=Get GravityOverride(), ExternalForce="0,0,0") -> (Pose) {
    angularDampingOverride=1
    angularSpringConstant=500
    bAngularSpring=true
    bGravityOverrideInSimSpace=true
    bOverrideAngularDamping=true
    bUseGravityOverride=true
    bUseSphericalLimits=true
    simulationSpace=RootRelative
    alphaBoolBlend {blendOption=Linear}
    boundBone {boneName=dyn_up_jacket_02_l}
    componentAppliedLinearAccClamp=<5000 5000 5000>
    componentLinearAccScale=<0.55 0.55 0.55>
    componentLinearVelScale=<0.5 0.5 0.5>
    gravityOverride {z=-980}
    physicsBodyDefinitions [{
      collisionType=OuterSphere
      boundBone {boneName=dyn_up_jacket_02_l}
      boxExtents=<6 6 6>
      constraintSetup {angularLimitsMax=<0 0 5>, angularLimitsMin=<-10 -10 -5>, angularTarget=<1 0 0>}
      localJointOffset=<-1 0 0>
     }]
    relativeSpaceBone {boneName=spine_04}
    sphericalLimits [{limitRadius=8, drivingBone {boneName=head}, sphereLocalOffset=<0 1.6607 8.0756>}]
   }
   "Transform (Modify) Bone - Bone: base_up_jacket_l"(Alpha=1, Translation="0,0,0", Rotation="0,0,-50.843", Scale="1,1,1") {
    rotationMode=BMM_Additive
    rotationSpace=BCS_BoneSpace
    alphaBoolBlend {blendOption=Linear}
    boneToModify {boneName=base_up_jacket_l}
    rotation {roll=-50.843}
   }
   "Transform (Modify) Bone - Bone: base_Collar_back_jacket"(Alpha=1, Translation="0,0,0", Rotation="-10,6,-6.715", Scale="1,1,1") {
    rotationMode=BMM_Additive
    rotationSpace=BCS_BoneSpace
    alphaBoolBlend {blendOption=Linear}
    boneToModify {boneName=base_Collar_back_jacket}
    rotation=<rot -10 6 -6.715>
   }
   n3 = Use cached pose 'Cape cache'()
   n4 = "Use cached pose 'deform cache (no RBan)'"()
   n5 = Get is Wearing Cape() -> (is Wearing Cape:bool)
   Output Pose(Result=Blend Poses by bool(
     BlendPose_0=Use cached pose 'cache for emote'()
     BlendPose_1=Blend Poses by bool(BlendPose_0=@n3, BlendPose_1=@n4, bActiveValue=@n5) {blendType=Linear}
     bActiveValue=Get bIsPlayingEmote()
    ) {blendType=Linear}) {layerGroup=None}
   n6 = Control Rig!(Source=Copy Pose From Mesh() {bUseAttachedParent=true}, Alpha, Get IsDBNO:bool=Get bIsDBNO()) -> (Pose) {
    alphaBoolBlend {blendOption=Linear}
    controlRigAssetReference {blueprintRigClass=&/F_MED_BrightDisk_CtrlRi.F_MED_BrightDisk_CtrlRi_C}
   }
   n7 = Anim Dynamics - Bone: dyn_boots_zip_r_1!(
    SimulationSpace:AnimPhysSimSpaceType Enum=BoneRelative
    ComponentPose=Local To Component(LocalPose=@n6)
    Alpha=1
    GravityOverride:Vector="0,0,-980"
    ExternalForce="0,0,0"
   ) -> (Pose) {
    angularBiasOverride=0.99
    angularDampingOverride=0.99
    angularSpringConstant=40
    bAngularSpring=true
    bLinearSpring=true
    bOverrideAngularBias=true
    bOverrideAngularDamping=true
    bOverrideLinearDamping=true
    bUseGravityOverride=true
    bUsePlanarLimit=false
    linearDampingOverride=0.99
    linearSpringConstant=40
    simulationSpace=BoneRelative
    alphaBoolBlend {blendOption=Linear}
    boundBone {boneName=dyn_boots_zip_r_1}
    componentAppliedLinearAccClamp=<150 150 150>
    componentLinearAccScale=<0.1 0.75 0.75>
    componentLinearVelScale=<0.1 0.75 0.75>
    physicsBodyDefinitions [{
      collisionType=CustomSphere
      boundBone {boneName=dyn_boots_zip_r_1}
      boxExtents=<9 2.7 5.3>
      constraintSetup {angularLimitsMax=<1 80 80>, angularLimitsMin=<0 0 0>, angularTarget=<0 0 0>}
      localJointOffset=<1 0 0>
     }]
    relativeSpaceBone {boneName=root}
    sphericalLimits [{limitRadius=1, drivingBone {boneName=calf_twist_02_r}, sphereLocalOffset=<0 0 0>}]
   }
   n8 = Anim Dynamics - Bone: dyn_boots_zip_l_1!(SimulationSpace:AnimPhysSimSpaceType Enum=BoneRelative, ComponentPose=@n7, Alpha=1, GravityOverride:Vector="0,0,-980", ExternalForce="0,0,0") -> (Pose) {
    angularBiasOverride=0.99
    angularDampingOverride=0.99
    angularSpringConstant=40
    bAngularSpring=true
    bLinearSpring=true
    bOverrideAngularBias=true
    bOverrideAngularDamping=true
    bOverrideLinearDamping=true
    bUseGravityOverride=true
    bUsePlanarLimit=false
    linearDampingOverride=0.99
    linearSpringConstant=40
    simulationSpace=BoneRelative
    alphaBoolBlend {blendOption=Linear}
    boundBone {boneName=dyn_boots_zip_l_1}
    componentAppliedLinearAccClamp=<150 150 150>
    componentLinearAccScale=<0.1 0.75 0.75>
    componentLinearVelScale=<0.1 0.75 0.75>
    physicsBodyDefinitions [{
      collisionType=CustomSphere
      boundBone {boneName=dyn_boots_zip_l_1}
      boxExtents=<9 2.7 5.3>
      constraintSetup {angularLimitsMax=<1 50 50>, angularLimitsMin=<0 0 0>, angularTarget=<0 0 0>}
      localJointOffset=<1 0 0>
     }]
    relativeSpaceBone {boneName=root}
    sphericalLimits [{limitRadius=1, drivingBone {boneName=calf_twist_02_r}, sphereLocalOffset=<0 0 0>}]
   }
   n9 = Anim Dynamics - Bone: dyn_zip!(ComponentPose=@n8, Alpha=1, GravityOverride:Vector=Get GravityOverride(), ExternalForce="0,0,0") -> (Pose) {
    angularBiasOverride=0.99
    angularDampingOverride=0.99
    angularSpringConstant=40
    bOverrideAngularDamping=true
    bUseGravityOverride=true
    bUsePlanarLimit=false
    linearDampingOverride=0.99
    linearSpringConstant=40
    simulationSpace=RootRelative
    alphaBoolBlend {blendOption=Linear}
    boundBone {boneName=dyn_zip}
    componentAppliedLinearAccClamp=<150 150 150>
    componentLinearAccScale=<0.1 0.75 0.75>
    componentLinearVelScale=<0.1 0.75 0.75>
    physicsBodyDefinitions [{
      collisionType=CustomSphere
      boundBone {boneName=dyn_zip}
      boxExtents=<10 10 10>
      constraintSetup {angularLimitsMax=<1 20 50>, angularLimitsMin=<-1 -80 -10>, angularTarget=<0 0 0>}
      localJointOffset=<2 0 0>
     }]
    relativeSpaceBone {boneName=root}
    sphericalLimits [{limitRadius=1, drivingBone {boneName=calf_twist_02_r}, sphereLocalOffset=<0 0 0>}]
   }
   n10 = "Transform (Modify) Bone - Bone: base_up_jacket_r"(ComponentPose=@n9, Alpha=Get head_left_right_rot() -> (head_left_right_rot:double), Translation="0,0,0", Rotation="0,20,0", Scale="1,1,1") {
    rotationMode=BMM_Additive
    rotationSpace=BCS_BoneSpace
    alphaBoolBlend {blendOption=Linear}
    boneToModify {boneName=base_up_jacket_r}
    rotation {yaw=20}
   }
   n11 = RigidBody!(ComponentPose=@n10, Alpha, OverrideWorldGravity:Vector=Get GravityOverride(), ExternalForce) -> (Pose) {
    bClampLinearTranslationLimitToRefPose=true
    bOverrideWorldGravity=true
    overridePhysicsAsset=&/F_MED_BrightDisk_Common_phys
    simulationSpace=BaseBoneSpace
    alphaBoolBlend {blendOption=Linear}
    baseBoneRef {boneName=spine_05}
    componentAppliedLinearAccClamp=<5000 5000 5000>
    componentLinearAccScale {z=0.6}
    componentLinearVelScale {z=0.6}
    overrideWorldGravity {z=-980}
    simSpaceSettings {
     maxAngularAcceleration=250
     maxAngularVelocity=3
     maxLinearAcceleration=500
     maxLinearVelocity=500
     velocityScaleZ=0
     worldAlpha=1
     externalLinearDragV=<1 1>
    }
   }
   n12 = "Transform (Modify) Bone - Bone: upperarm_l"!(ComponentPose=@n11, bAlphaBoolEnabled:bool=True, Translation="0,0,0", Rotation="2.2,-3.2,0", Scale="1,1,1") -> (Pose) {
    alphaInputType=Bool
    rotationMode=BMM_Additive
    rotationSpace=BCS_BoneSpace
    alphaBoolBlend {blendOption=Linear}
    boneToModify {boneName=upperarm_l}
    rotation {pitch=2.2, yaw=-3.2}
   }
   "Save cached pose 'deform cache (no RBan)'"(Pose=Component To Local(ComponentPose=@n12)) {cacheName="deform cache (no RBan)"}
   n13 = RigidBody!(ComponentPose=Local To Component(LocalPose="Use cached pose 'deform cache (no RBan)'"()), Alpha, ExternalForce) -> (Pose) {
    overridePhysicsAsset=&/F_MED_BrightDisk_CapeColliders
    alphaBoolBlend {blendOption=Linear}
   }
   Save cached pose 'Cape cache'(Pose=Component To Local(ComponentPose=@n13)) {cacheName=Cape cache}
   n14 = Control Rig!(Source=Copy Pose From Mesh() {bUseAttachedParent=true}, Alpha, is emote:bool=Get bIsPlayingEmote()) -> (Pose) {
    alphaBoolBlend {blendOption=Linear}
    controlRigAssetReference {blueprintRigClass=&/F_MED_BrightDisk_CtrlRig_Emote.F_MED_BrightDisk_CtrlRig_Emote_C}
   }
   n15 = Anim Dynamics - Bone: dyn_boots_zip_r_1(SimulationSpace=BoneRelative, ComponentPose=Local To Component(LocalPose=@n14), Alpha=1, GravityOverride="0,0,-980", ExternalForce="0,0,0") {
    angularBiasOverride=0.99
    angularDampingOverride=0.99
    angularSpringConstant=40
    bAngularSpring=true
    bLinearSpring=true
    bOverrideAngularBias=true
    bOverrideAngularDamping=true
    bOverrideLinearDamping=true
    bUseGravityOverride=true
    bUsePlanarLimit=false
    linearDampingOverride=0.99
    linearSpringConstant=40
    simulationSpace=BoneRelative
    alphaBoolBlend {blendOption=Linear}
    boundBone {boneName=dyn_boots_zip_r_1}
    componentAppliedLinearAccClamp=<150 150 150>
    componentLinearAccScale=<0.1 0.75 0.75>
    componentLinearVelScale=<0.1 0.75 0.75>
    physicsBodyDefinitions [{
      collisionType=CustomSphere
      boundBone {boneName=dyn_boots_zip_r_1}
      boxExtents=<9 2.7 5.3>
      constraintSetup {angularLimitsMax=<1 80 80>, angularLimitsMin=<0 0 0>, angularTarget=<0 0 0>}
      localJointOffset=<1 0 0>
     }]
    relativeSpaceBone {boneName=root}
    sphericalLimits [{limitRadius=1, drivingBone {boneName=calf_twist_02_r}, sphereLocalOffset=<0 0 0>}]
   }
   n16 = Anim Dynamics - Bone: dyn_boots_zip_l_1(SimulationSpace=BoneRelative, ComponentPose=@n15, Alpha=1, GravityOverride="0,0,-980", ExternalForce="0,0,0") {
    angularBiasOverride=0.99
    angularDampingOverride=0.99
    angularSpringConstant=40
    bAngularSpring=true
    bLinearSpring=true
    bOverrideAngularBias=true
    bOverrideAngularDamping=true
    bOverrideLinearDamping=true
    bUseGravityOverride=true
    bUsePlanarLimit=false
    linearDampingOverride=0.99
    linearSpringConstant=40
    simulationSpace=BoneRelative
    alphaBoolBlend {blendOption=Linear}
    boundBone {boneName=dyn_boots_zip_l_1}
    componentAppliedLinearAccClamp=<150 150 150>
    componentLinearAccScale=<0.1 0.75 0.75>
    componentLinearVelScale=<0.1 0.75 0.75>
    physicsBodyDefinitions [{
      collisionType=CustomSphere
      boundBone {boneName=dyn_boots_zip_l_1}
      boxExtents=<9 2.7 5.3>
      constraintSetup {angularLimitsMax=<1 50 50>, angularLimitsMin=<0 0 0>, angularTarget=<0 0 0>}
      localJointOffset=<1 0 0>
     }]
    relativeSpaceBone {boneName=root}
    sphericalLimits [{limitRadius=1, drivingBone {boneName=calf_twist_02_r}, sphereLocalOffset=<0 0 0>}]
   }
   n17 = Anim Dynamics - Bone: dyn_zip!(SimulationSpace:AnimPhysSimSpaceType Enum, ComponentPose=@n16, Alpha=1, GravityOverride="0,0,-980", ExternalForce="0,0,0") -> (Pose) {
    angularBiasOverride=0.99
    angularDampingOverride=0.99
    angularSpringConstant=40
    bAngularSpring=true
    bLinearSpring=true
    bOverrideAngularBias=true
    bOverrideAngularDamping=true
    bOverrideLinearDamping=true
    bUseGravityOverride=true
    bUsePlanarLimit=false
    linearDampingOverride=0.99
    linearSpringConstant=40
    simulationSpace=BoneRelative
    alphaBoolBlend {blendOption=Linear}
    boundBone {boneName=dyn_zip}
    componentAppliedLinearAccClamp=<150 150 150>
    componentLinearAccScale=<0.1 0.75 0.75>
    componentLinearVelScale=<0.1 0.75 0.75>
    physicsBodyDefinitions [{
      collisionType=CustomSphere
      boundBone {boneName=dyn_zip}
      boxExtents=<9 2.7 5.3>
      constraintSetup {angularLimitsMax=<1 50 50>, angularLimitsMin=<0 0 0>, angularTarget=<0 0 0>}
      localJointOffset=<1 0 0>
     }]
    relativeSpaceBone {boneName=root}
    sphericalLimits [{limitRadius=1, drivingBone {boneName=calf_twist_02_r}, sphereLocalOffset=<0 0 0>}]
   }
   n18 = RigidBody!(ComponentPose=@n17, Alpha, OverrideWorldGravity:Vector=Get GravityOverride(), ExternalForce) -> (Pose) {
    bClampLinearTranslationLimitToRefPose=true
    bOverrideWorldGravity=true
    overridePhysicsAsset=&/F_MED_BrightDisk_EMOTE
    simulationSpace=BaseBoneSpace
    alphaBoolBlend {blendOption=Linear}
    baseBoneRef {boneName=spine_05}
    componentAppliedLinearAccClamp=<5000 5000 5000>
    componentLinearAccScale {z=0.6}
    componentLinearVelScale {z=0.6}
    overrideWorldGravity {z=-980}
    simSpaceSettings {
     maxAngularAcceleration=500
     maxAngularVelocity=3
     maxLinearAcceleration=700
     maxLinearVelocity=500
     velocityScaleZ=0.75
     worldAlpha=0.25
     externalLinearDragV=<1 1>
    }
   }
   Save cached pose 'cache for emote'(Pose=Component To Local(ComponentPose=@n18)) {cacheName=cache for emote}
  ]}
 EventGraph {nodes [
   Event BlueprintUpdateAnimation()
   > n1 = Sequence!(execute) -> (then_0, then_1, then_2:Exec)
   > n2 = Gravity Override State Function DEPRECATED(CharacterPartAnimInstance=Self-Reference(), GameTimeInSeconds=GetGameTimeInSeconds(), GravityOverrideParams=Get Gravity Override Params())
   > n3 = Set GravityOverride(GravityOverride=@n2)
   n4 = Get bIsSkydiveDiveMode()
   > Set GravityOverrideChain(GravityOverrideChain:Vector=SelectVector(A="0,-200,-600", B=SelectVector(A="-750,0,0", B=@n3, bPickA=@n4), bPickA=Get bIsParachuteOpened())) -> (Output_Get:Vector)
   n5 = Get bIsPlayingEmote()
   > Set FrontBOOL(FrontBOOL:bool=AND Boolean(A=Get bIsInFrontEnd(), B=NOT Boolean(A=@n5))) -> (Output_Get:bool)
   n6 = CheckForBackbling()
   Set is Wearing Backbling(
    execute=@n1.then_1
    is Wearing Backbling:bool=AND Boolean!(A=@n6, B=NOT Boolean(A=@n6.bIsBackblingACape), C:bool=NOT Boolean(A=@n6.bIsBackblingLong)) -> (ReturnValue)
   ) -> (Output_Get:bool)
   > Set is Wearing Cape(is Wearing Cape:bool=OR Boolean(A=@n6.bIsBackblingACape, B=@n6.bIsBackblingLong)) -> (Output_Get:bool)
   n7 = Branch(execute=@n1.then_2)
   n8 = BreakTransform(InTransform=GetDeltaTransformFromRefPose(self=GetOwnerSkeletalMeshComponent(), BoneName=head, BaseName=neck_01))
   > Set head_left_right_rot(head_left_right_rot:double=MapRangeClamped(Value=BreakRotator(InRot=@n8.Rotation), InRangeA=1, InRangeB=7)) -> (Output_Get:double)
   Set head_left_right_rot(execute=@n7.else, head_left_right_rot=0)
   "Nearly Equal (Float)"(A="To Float (Boolean)"(InBool=Get is Wearing Backbling() -> (is Wearing Backbling:bool)), B=1, ErrorTolerance=0.1)
   Get bIsMoving2D()
  ]}
]
variables [
 Anim_Dynamics_Alpha {type=number}
 Custom Base Mesh {type=SkeletalMeshComponent}
 FrontBOOL {type=boolean}
 Gravity Override Params {type=GravityOverrideParamsStruct}
 GravityOverride {type=Vector, default=<0 0 -980>}
 GravityOverrideChain {type=Vector}
 head_left_right_rot {type=number}
 is Wearing Backbling {type=boolean}
 is Wearing Cape {type=boolean}
]
