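# Physics asset for the hair tail on the BrightDisk female medium face accessory, apparently for secondary motion.
# Capsule bodies form a six-bone tail chain from base_hairTail to hairTail_05, joined by tight twist limits and wider swing limits past the first link.
# Capsules on head, upper spine and arms, linked by free angular constraints, provide body collision volumes.
# The tail base has both a locked and a free constraint to the head.
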PhysicsAsset /BRCosmetics/Characters/Player/Female/Medium/Bodies/F_MED_BrightDisk/Meshes/Parts/F_MED_BrightDisk_FaceAcc_SECOND
bodies {
 base_hairTail [{shapeType=Capsule, length=0.5, radius=0.5, shapeName=base_hairTail_capsule}]
 hairTail_01 [{shapeType=Capsule, length=3.8667, radius=2.1333}]
 hairTail_02 [{shapeType=Capsule, length=1, radius=6.2685, center=<4.1793 0 0>, rotation=<rot -90 0 0>}]
 hairTail_03 [{shapeType=Capsule, length=1, radius=5, center=<3.9727 -1.2951 0>}]
 hairTail_04 [
  {shapeType=Capsule, length=0, radius=6.337, center=<4.6909 -2.4859 0.0326>, rotation=<rot -89.1062 179.9989 -179.5966>}
  {
   shapeType=Capsule
   length=11.2653
   radius=2.1136
   shapeName=hairTail_04_capsule
   center=<4.0407 1.8127 -4.8973>
   rotation=<rot 86.6597 0 -0.0002>
  }
 ]
 hairTail_05 [{
   shapeType=Capsule
   length=3.4025
   radius=4.1535
   shapeName=hairTail_05_capsule
   center=<4.9353 -1.4632 0.0093>
   rotation=<rot 85.462 0 5.3678>
  }]
 head [{shapeType=Capsule, length=23.0665, radius=11.3669, center=<3.9731 3.3438 -0.3407>, rotation=<rot 1.7486 -95.6426 -0.1134>}]
 lowerarm_l [{
   shapeType=Capsule
   length=20.0048
   radius=7.6497
   shapeName=lowerarm_l_capsule
   center=<9.2362 -4.5038 0.9904>
   rotation=<rot 3.2123 -116.305 96.4818>
  }]
 lowerarm_r [{
   shapeType=Capsule
   length=18.4145
   radius=9.24
   shapeName=lowerarm_l_capsule
   center=<-9.2362 4.5038 -0.9904>
   rotation=<rot 3.2123 -116.305 -83.5182>
  }]
 spine_03 [{shapeType=Capsule, length=17.7177, radius=7.7556, shapeName=spine_03_capsule, center=<3.4828 0 0>, rotation=<rot 0 -90 0>}]
 spine_04 [{shapeType=Capsule, length=17.8108, radius=8.0523, shapeName=spine_04_capsule, center=<9.4659 1.1006 0>}]
 spine_05 [{shapeType=Capsule, length=17.0972, radius=11.8124, shapeName=spine_05_capsule, center=<8.9515 5.2 0>, rotation=<rot 0 4.7836 0>}]
 upperarm_l [
  {
   shapeType=Capsule
   length=21.92
   radius=9.1512
   shapeName=upperarm_l_capsule
   center=<15.6417 0.0001 2.0648>
   rotation=<rot 0 -89.9994 -88.9149>
  }
  {shapeType=Capsule, length=3.1565, radius=9.3433, shapeName=upperarm_l_capsule, center=<2.8425 0 2.6099>, rotation=<rot 41.2485 0 -0.0005>}
 ]
 upperarm_r [
  {
   shapeType=Capsule
   length=21.6022
   radius=9.4689
   shapeName=upperarm_l_capsule
   center=<-18.677 -0.0001 -1.9579>
   rotation=<rot 0 -89.9997 91.263>
  }
  {shapeType=Capsule, length=3.1565, radius=9.3433, shapeName=upperarm_r_capsule, center=<-2.5496 0 -2.341>, rotation=<rot 40.7066 0 -0.0004>}
 ]
}
constraints [
 {bone1Name=hairTail_01, bone2Name=base_hairTail, swing1LimitDegrees=10, swing2LimitDegrees=10, twistLimitDegrees=5}
 {bone1Name=hairTail_02, bone2Name=hairTail_01, swing1LimitDegrees=45, swing2LimitDegrees=25, twistLimitDegrees=5}
 {bone1Name=hairTail_03, bone2Name=hairTail_02, swing1LimitDegrees=40, swing2LimitDegrees=25, twistLimitDegrees=5}
 {bone1Name=hairTail_04, bone2Name=hairTail_03, swing1LimitDegrees=40, swing2LimitDegrees=20, twistLimitDegrees=5}
 {
  bone1Name=head
  bone2Name=spine_05
  swing1LimitDegrees=45
  swing1Motion=Free
  swing2LimitDegrees=45
  swing2Motion=Free
  twistLimitDegrees=45
  twistMotion=Free
 }
 {
  bone1Name=spine_05
  bone2Name=spine_04
  swing1LimitDegrees=45
  swing1Motion=Free
  swing2LimitDegrees=45
  swing2Motion=Free
  twistLimitDegrees=45
  twistMotion=Free
 }
 {bone1Name=hairTail_05, bone2Name=hairTail_04, swing1LimitDegrees=33, swing2LimitDegrees=20, twistLimitDegrees=5}
 {
  bone1Name=head
  bone2Name=base_hairTail
  swing1LimitDegrees=45
  swing1Motion=Free
  swing2LimitDegrees=45
  swing2Motion=Free
  twistLimitDegrees=45
  twistMotion=Free
 }
 {
  bone1Name=base_hairTail
  bone2Name=head
  swing1LimitDegrees=45
  swing1Motion=Locked
  swing2LimitDegrees=45
  swing2Motion=Locked
  twistLimitDegrees=45
  twistMotion=Locked
 }
 {
  bone1Name=upperarm_l
  bone2Name=spine_05
  swing1LimitDegrees=45
  swing1Motion=Free
  swing2LimitDegrees=45
  swing2Motion=Free
  twistLimitDegrees=45
  twistMotion=Free
 }
 {
  bone1Name=upperarm_r
  bone2Name=spine_05
  swing1LimitDegrees=45
  swing1Motion=Free
  swing2LimitDegrees=45
  swing2Motion=Free
  twistLimitDegrees=45
  twistMotion=Free
 }
 {
  bone1Name=lowerarm_l
  bone2Name=upperarm_l
  swing1LimitDegrees=45
  swing1Motion=Free
  swing2LimitDegrees=45
  swing2Motion=Free
  twistLimitDegrees=45
  twistMotion=Free
 }
 {
  bone1Name=lowerarm_r
  bone2Name=upperarm_r
  swing1LimitDegrees=45
  swing1Motion=Free
  swing2LimitDegrees=45
  swing2Motion=Free
  twistLimitDegrees=45
  twistMotion=Free
 }
 {
  bone1Name=spine_04
  bone2Name=spine_03
  swing1LimitDegrees=45
  swing1Motion=Free
  swing2LimitDegrees=45
  swing2Motion=Free
  twistLimitDegrees=45
  twistMotion=Free
 }
 {
  bone1Name=base_hairTail
  bone2Name=spine_03
  swing1LimitDegrees=45
  swing1Motion=Free
  swing2LimitDegrees=45
  swing2Motion=Free
  twistLimitDegrees=45
  twistMotion=Free
 }
]
