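# Animation Blueprint for the BrightDisk female FaceAcc part, adding secondary motion to the character's hair.
# It copies parent and Body poses, then runs AnimDynamics on side and front hair.
# With backbling or capes, the hair tail uses bone constraints and head-rotation-driven offsets; otherwise it uses RigidBody physics when enabled, falling back to similar procedural posing.
# The EventGraph updates state-based gravity, backbling flags and head rotation, which is zeroed without regular backbling.
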AnimBlueprint /BRCosmetics/Characters/Player/Female/Medium/Bodies/F_MED_BrightDisk/Meshes/Parts/F_MED_BrightDisk_FaceAcc_AnimBP
target_skeleton=&/F_MED_BrightDisk_FaceAcc_Skeleton
graphs [
 AnimGraph {nodes [
   "Constraint - hairTail_01 (PP)"!(ComponentPose, Alpha=1, ConstraintWeights_0=0.8, ConstraintWeights_1:float=0.2) -> (Pose) {
    alphaBoolBlend {blendOption=Linear}
    boneToModify {boneName=hairTail_01}
    constraintSetup [{targetBone {boneName=head}}, {targetBone {boneName=spine_05}}]
    constraintWeights [1,1]
   }
   "Constraint - hairTail_05 (PPP)"!(ComponentPose, Alpha=0.9, ConstraintWeights_0=0.3, ConstraintWeights_1:float=0.5, ConstraintWeights_2:float=0.2) -> (Pose) {
    alphaBoolBlend {blendOption=Linear}
    boneToModify {boneName=hairTail_05}
    constraintSetup [{targetBone {boneName=head}}, {targetBone {boneName=spine_05}}, {targetBone {boneName=spine_04}}]
    constraintWeights [1,1,1]
   }
   "Constraint - hairTail_04 (PPP)"!(ComponentPose, Alpha=1, ConstraintWeights_0=0.3, ConstraintWeights_1:float=0.5, ConstraintWeights_2:float=0.2) -> (Pose) {
    alphaBoolBlend {blendOption=Linear}
    boneToModify {boneName=hairTail_04}
    constraintSetup [{targetBone {boneName=head}}, {targetBone {boneName=spine_05}}, {targetBone {boneName=spine_04}}]
    constraintWeights [1,1,1]
   }
   "Constraint - base_hairTail (PP)"!(ComponentPose, Alpha=1, ConstraintWeights_0=0.9, ConstraintWeights_1:float=0.1) -> (Pose) {
    alphaBoolBlend {blendOption=Linear}
    boneToModify {boneName=base_hairTail}
    constraintSetup [{targetBone {boneName=head}}, {targetBone {boneName=spine_05}}]
    constraintWeights [1,1]
   }
   n1 = Use cached pose 'Normal Dynamics'()
   n2 = Use cached pose 'BackPack Dynamicss'()
   n3 = Get is Wearing Cape() -> (is Wearing Cape:bool)
   Output Pose(Result=Blend Poses by bool(
     BlendPose_0=Use cached pose 'BackPack Dynamicss'()
     BlendPose_1=Blend Poses by bool(BlendPose_0=@n2, BlendPose_1=@n1, bActiveValue=@n3) {blendType=Linear}
     bActiveValue=Get is Wearing Backbling() -> (is Wearing Backbling:bool)
    ) {blendType=Linear}) {layerGroup=None}
   n4 = "Constraint - hairTail_04 (PPP)"(
    ComponentPose=Local To Component(LocalPose=Use cached pose 'StartPose'())
    Alpha=0.8
    ConstraintWeights_0=0.3
    ConstraintWeights_1=0.5
    ConstraintWeights_2=0.2
   ) {
    alphaBoolBlend {blendOption=Linear}
    boneToModify {boneName=hairTail_04}
    constraintSetup [{targetBone {boneName=head}}, {targetBone {boneName=spine_05}}, {targetBone {boneName=spine_04}}]
    constraintWeights [1,1,1]
   }
   n5 = "Constraint - hairTail_05 (PPP)"(ComponentPose=@n4, Alpha=0.9, ConstraintWeights_0=0.2, ConstraintWeights_1=0.5, ConstraintWeights_2=0.3) {
    alphaBoolBlend {blendOption=Linear}
    boneToModify {boneName=hairTail_05}
    constraintSetup [{targetBone {boneName=head}}, {targetBone {boneName=spine_05}}, {targetBone {boneName=spine_04}}]
    constraintWeights [1,1,1]
   }
   n6 = Get head_fwdBwd_rot() -> (head_fwdBwd_rot:double)
   n7 = "Transform (Modify) Bone - Bone: hairTail_01"(ComponentPose=@n5, Alpha=@n6, Translation="1,0,0.678", Rotation="0.002,-4.009,0.001", Scale="1,1,1") {
    rotationMode=BMM_Additive
    rotationSpace=BCS_BoneSpace
    translationMode=BMM_Additive
    translationSpace=BCS_BoneSpace
    alphaBoolBlend {blendOption=Linear}
    alphaScaleBiasClamp {bMapRange=true, inRange {max=0, min=-25}, outRange {max=0, min=1}}
    boneToModify {boneName=hairTail_01}
    rotation=<rot 0.002 -4.0088 0.001>
    translation {x=1, z=0.678}
   }
   n8 = "Transform (Modify) Bone - Bone: hairTail_02"(ComponentPose=@n7, Alpha=@n6, Translation="-0.001,-0.467,-0.240", Rotation="-0.01,0,0.003", Scale="1,1,1") {
    rotationMode=BMM_Additive
    rotationSpace=BCS_BoneSpace
    translationMode=BMM_Additive
    translationSpace=BCS_BoneSpace
    alphaBoolBlend {blendOption=Linear}
    alphaScaleBiasClamp {bMapRange=true, inRange {max=0, min=-25}, outRange {max=0, min=1}}
    boneToModify {boneName=hairTail_02}
    rotation {pitch=-0.01, roll=0.003}
    translation=<-0.001 -0.467 -0.24>
   }
   n9 = Get head_fwdBwd_rot()
   n10 = "Transform (Modify) Bone - Bone: hairTail_01"(ComponentPose=@n8, Alpha=@n9, Translation="0,1,0", Rotation="0,0,0", Scale="1,1,1") {
    rotationMode=BMM_Additive
    rotationSpace=BCS_BoneSpace
    translationMode=BMM_Additive
    translationSpace=BCS_BoneSpace
    alphaBoolBlend {blendOption=Linear}
    alphaScaleBiasClamp {bMapRange=true, inRange {max=15}}
    boneToModify {boneName=hairTail_01}
    translation {y=1}
   }
   n11 = "Transform (Modify) Bone - Bone: hairTail_02"(ComponentPose=@n10, Alpha=@n9, Translation="0.009,1,0", Rotation="0,0,0", Scale="1,1,1") {
    rotationMode=BMM_Additive
    rotationSpace=BCS_BoneSpace
    translationMode=BMM_Additive
    translationSpace=BCS_BoneSpace
    alphaBoolBlend {blendOption=Linear}
    alphaScaleBiasClamp {bMapRange=true, inRange {max=15}}
    boneToModify {boneName=hairTail_02}
    translation=<0.009 1>
   }
   n12 = "Transform (Modify) Bone - Bone: hairTail_04"(ComponentPose=@n11, Alpha=@n9, Translation="0.009,1,0", Rotation="-0.009,11.852,0.002", Scale="1,1,1") {
    rotationMode=BMM_Additive
    rotationSpace=BCS_BoneSpace
    translationMode=BMM_Additive
    translationSpace=BCS_BoneSpace
    alphaBoolBlend {blendOption=Linear}
    alphaScaleBiasClamp {bMapRange=true, inRange {max=15}}
    boneToModify {boneName=hairTail_04}
    rotation=<rot -0.009 11.8521 0.002>
    translation=<0.009 1>
   }
   n13 = RigidBody!(
    ComponentPose=Local To Component(LocalPose=Use cached pose 'StartPose'())
    Alpha
    OverrideWorldGravity:Vector=Get GravityOverride()
    ExternalForce
   ) -> (Pose) {
    bClampLinearTranslationLimitToRefPose=true
    bOverrideWorldGravity=true
    overridePhysicsAsset=&/F_MED_BrightDisk_FaceAcc_SECOND
    simulationSpace=BaseBoneSpace
    alphaBoolBlend {blendOption=Linear}
    baseBoneRef {boneName=head}
    componentAppliedLinearAccClamp=<3000 3000 3000>
    componentLinearAccScale=<0.4 0.3 0.3>
    componentLinearVelScale=<0.4 0.3 0.3>
    overrideWorldGravity {z=-980}
    simSpaceSettings {maxAngularAcceleration=200, maxAngularVelocity=2, maxLinearAcceleration=320, maxLinearVelocity=250, velocityScaleZ=0, worldAlpha=0.712}
   }
   Save cached pose 'Normal Dynamics'(Pose=Blend Poses by bool(BlendPose_0=Component To Local(ComponentPose=@n13), BlendPose_1=Component To Local(ComponentPose=@n12), bActiveValue=Get bRigidBodyEnabled()) {blendType=Linear}) {cacheName=Normal Dynamics}
   n14 = Copy Pose From Mesh() {bCopyCurves=true, bUseAttachedParent=true}
   n15 = Copy Pose From Mesh(SourceMeshComponent=Get Body() -> (Body:Skeletal Mesh Component Object Reference))
   n16 = Anim Dynamics - Bone: hairSide_r_00!(
    ComponentPose=Local To Component(LocalPose=Layered blend per bone(BasePose=@n14, BlendPoses_0=@n15) {layerSetup [{branchFilters [{blendDepth=2, boneName=clavicle_r}, {blendDepth=2, boneName=clavicle_l}]}]})
    Alpha=1
    GravityOverride:Vector="0,0,-980"
    ExternalForce="0,0,0"
   ) -> (Pose) {
    angularDampingOverride=0.9999
    angularSpringConstant=10
    bAngularSpring=true
    bOverrideAngularDamping=true
    bUseGravityOverride=true
    linearDampingOverride=0.99
    simulationSpace=RootRelative
    alphaBoolBlend {blendOption=Linear}
    boundBone {boneName=hairSide_r_00}
    gravityOverride {z=-980}
    physicsBodyDefinitions [{
      boundBone {boneName=hairSide_r_00}
      boxExtents=<7.5 7.5 7.5>
      constraintSetup {angularLimitsMax=<10 50 80>, angularLimitsMin=<-10 -3 -7>, angularTarget=<0 0 0>}
      localJointOffset=<0 0 0>
     }]
    relativeSpaceBone {boneName=head}
    sphericalLimits [{limitRadius=7.8476, drivingBone {boneName=head}, sphereLocalOffset=<1.3801 2.1965 -1.6179>}]
   }
   n17 = Anim Dynamics - Bone: hairSide_l_01!(ComponentPose=@n16, Alpha=1, GravityOverride:Vector="0,0,-980", ExternalForce="0,0,0") -> (Pose) {
    angularDampingOverride=0.9999
    angularSpringConstant=10
    bAngularSpring=true
    bOverrideAngularDamping=true
    bUseGravityOverride=true
    linearDampingOverride=0.99
    alphaBoolBlend {blendOption=Linear}
    boundBone {boneName=hairSide_l_01}
    gravityOverride {z=-980}
    physicsBodyDefinitions [{
      boundBone {boneName=hairSide_l_01}
      boxExtents=<7.5 7.5 7.5>
      constraintSetup {angularLimitsMax=<10 30 17>, angularLimitsMin=<-10 -3 -17>, angularTarget=<0 0 0>}
      localJointOffset=<1 0 0>
     }]
    relativeSpaceBone {boneName=head}
    sphericalLimits [{limitRadius=7.8476, drivingBone {boneName=head}, sphereLocalOffset=<4.5172 2.1965 -1.6179>}]
   }
   n18 = Anim Dynamics - Bone: hair_f_l_1!(ComponentPose=@n17, Alpha=1, GravityOverride:Vector="0,0,-980", ExternalForce="0,0,0") -> (Pose) {
    angularDampingOverride=0.9999
    angularSpringConstant=10
    bAngularSpring=true
    bOverrideAngularDamping=true
    bUseGravityOverride=true
    linearDampingOverride=0.99
    alphaBoolBlend {blendOption=Linear}
    boundBone {boneName=hair_f_l_1}
    gravityOverride {z=-980}
    physicsBodyDefinitions [{
      boundBone {boneName=hair_f_l_1}
      boxExtents=<7.5 7.5 7.5>
      constraintSetup {angularLimitsMax=<10 30 17>, angularLimitsMin=<-10 -3 -17>, angularTarget=<0 0 0>}
      localJointOffset=<1 0 0>
     }]
    relativeSpaceBone {boneName=head}
    sphericalLimits [{limitRadius=7.8476, drivingBone {boneName=head}, sphereLocalOffset=<4.5172 2.1965 -1.6179>}]
   }
   n19 = Anim Dynamics - Bone: hair_f_m_1!(ComponentPose=@n18, Alpha=1, GravityOverride:Vector="0,0,-980", ExternalForce="0,0,0") -> (Pose) {
    angularDampingOverride=0.9999
    angularSpringConstant=10
    bAngularSpring=true
    bOverrideAngularDamping=true
    bUseGravityOverride=true
    linearDampingOverride=0.99
    alphaBoolBlend {blendOption=Linear}
    boundBone {boneName=hair_f_m_1}
    gravityOverride {z=-980}
    physicsBodyDefinitions [{
      boundBone {boneName=hair_f_m_1}
      boxExtents=<7.5 7.5 7.5>
      constraintSetup {angularLimitsMax=<10 30 17>, angularLimitsMin=<-10 -3 0>, angularTarget=<0 0 0>}
      localJointOffset=<1 0 0>
     }]
    relativeSpaceBone {boneName=head}
    sphericalLimits [{limitRadius=7.8476, drivingBone {boneName=head}, sphereLocalOffset=<4.5172 2.874 -1.6179>}]
   }
   Save cached pose 'StartPose'(Pose=Component To Local(ComponentPose=@n19)) {cacheName=StartPose}
   n20 = "Constraint - hairTail_04 (PPP)"(
    ComponentPose=Local To Component(LocalPose=Use cached pose 'StartPose'())
    Alpha=0.8
    ConstraintWeights_0=0.3
    ConstraintWeights_1=0.5
    ConstraintWeights_2=0.2
   ) {
    alphaBoolBlend {blendOption=Linear}
    boneToModify {boneName=hairTail_04}
    constraintSetup [{targetBone {boneName=head}}, {targetBone {boneName=spine_05}}, {targetBone {boneName=spine_04}}]
    constraintWeights [1,1,1]
   }
   n21 = "Constraint - hairTail_05 (PPP)"(ComponentPose=@n20, Alpha=0.9, ConstraintWeights_0=0.2, ConstraintWeights_1=0.5, ConstraintWeights_2=0.3) {
    alphaBoolBlend {blendOption=Linear}
    boneToModify {boneName=hairTail_05}
    constraintSetup [{targetBone {boneName=head}}, {targetBone {boneName=spine_05}}, {targetBone {boneName=spine_04}}]
    constraintWeights [1,1,1]
   }
   n22 = Get head_fwdBwd_rot()
   n23 = "Transform (Modify) Bone - Bone: hairTail_01"(ComponentPose=@n21, Alpha=@n22, Translation="0.015,0,0.678", Rotation="0.002,0,0.001", Scale="1,1,1") {
    rotationMode=BMM_Additive
    rotationSpace=BCS_BoneSpace
    translationMode=BMM_Additive
    translationSpace=BCS_BoneSpace
    alphaBoolBlend {blendOption=Linear}
    alphaScaleBiasClamp {bMapRange=true, inRange {max=0, min=-25}, outRange {max=0, min=1}}
    boneToModify {boneName=hairTail_01}
    rotation {pitch=0.002, roll=0.001}
    translation {x=0.015, z=0.678}
   }
   n24 = "Transform (Modify) Bone - Bone: hairTail_02"(ComponentPose=@n23, Alpha=@n22, Translation="0,0,0", Rotation="-0.01,0,0", Scale="1,1,1") {
    rotationMode=BMM_Additive
    rotationSpace=BCS_BoneSpace
    translationMode=BMM_Additive
    translationSpace=BCS_BoneSpace
    alphaBoolBlend {blendOption=Linear}
    alphaScaleBiasClamp {bMapRange=true, inRange {max=0, min=-25}, outRange {max=0, min=1}}
    boneToModify {boneName=hairTail_02}
    rotation {pitch=-0.01}
   }
   n25 = Get head_fwdBwd_rot()
   n26 = "Transform (Modify) Bone - Bone: hairTail_01"(ComponentPose=@n24, Alpha=@n25, Translation="0,1,0", Rotation="0,0,0", Scale="1,1,1") {
    rotationMode=BMM_Additive
    rotationSpace=BCS_BoneSpace
    translationMode=BMM_Additive
    translationSpace=BCS_BoneSpace
    alphaBoolBlend {blendOption=Linear}
    alphaScaleBiasClamp {bMapRange=true, inRange {max=15}}
    boneToModify {boneName=hairTail_01}
    translation {y=1}
   }
   n27 = "Transform (Modify) Bone - Bone: hairTail_02"(ComponentPose=@n26, Alpha=@n25, Translation="0.009,1,0", Rotation="0,0,0", Scale="1,1,1") {
    rotationMode=BMM_Additive
    rotationSpace=BCS_BoneSpace
    translationMode=BMM_Additive
    translationSpace=BCS_BoneSpace
    alphaBoolBlend {blendOption=Linear}
    alphaScaleBiasClamp {bMapRange=true, inRange {max=15}}
    boneToModify {boneName=hairTail_02}
    translation=<0.009 1>
   }
   n28 = "Transform (Modify) Bone - Bone: hairTail_04"(ComponentPose=@n27, Alpha=@n25, Translation="0.009,1,0", Rotation="-0.009,11.852,0.002", Scale="1,1,1") {
    rotationMode=BMM_Additive
    rotationSpace=BCS_BoneSpace
    translationMode=BMM_Additive
    translationSpace=BCS_BoneSpace
    alphaBoolBlend {blendOption=Linear}
    alphaScaleBiasClamp {bMapRange=true, inRange {max=15}}
    boneToModify {boneName=hairTail_04}
    rotation=<rot -0.009 11.8521 0.002>
    translation=<0.009 1>
   }
   Save cached pose 'BackPack Dynamicss'(Pose=Component To Local(ComponentPose=@n28)) {cacheName=BackPack Dynamicss}
   Use cached pose 'Normal Dynamics'()
  ]}
 EventGraph {nodes [
   Event BlueprintUpdateAnimation()
   > n1 = Sequence!(execute) -> (then_0, then_1, then_2:Exec, then_3:Exec, then_4:Exec)
   > n2 = Gravity Override State Function DEPRECATED(CharacterPartAnimInstance=Self-Reference(), GameTimeInSeconds=GetGameTimeInSeconds(), GravityOverrideParams=Get Gravity Override Params())
   > Set GravityOverride(GravityOverride=@n2)
   n3 = SelectVector(A="0.2,0.2,0.2", B="0.35,0.35,0.35", bPickA=AND Boolean(A=Get bIsSkyDiving(), B=Get bIsFalling()))
   > Set ComponentLinearAccScale(ComponentLinearAccScale:Vector=@n3) -> (Output_Get:Vector)
   > Set ComponentLinearVelScale(ComponentLinearVelScale:Vector=@n3) -> (Output_Get:Vector)
   n4 = CheckForBackbling()
   Set is Wearing Backbling(
    execute=@n1.then_3
    is Wearing Backbling:bool=AND Boolean!(A=@n4, B=NOT Boolean(A=@n4.bIsBackblingACape), C:bool=NOT Boolean(A=@n4.bIsBackblingLong)) -> (ReturnValue)
   ) -> (Output_Get:bool)
   > Set is Wearing Cape(is Wearing Cape:bool=OR Boolean(A=@n4.bIsBackblingACape, B=@n4.bIsBackblingLong)) -> (Output_Get:bool)
   Set hair Rot(
    execute=@n1.then_2
    hair Rot:double=GetAlphaFromReferenceTransformDelta(
     SkeletalMeshComponent=Get body() -> (body:Skeletal Mesh Component Object Reference)
     BoneName=base_hairTail
     BaseName=head
     InputRange=(X=0,Y=1)
     TwistAxis="1,0,0"
    ).RawOutput
   ) -> (Output_Get:double)
   n5 = Get is Wearing Backbling()
   n6 = Branch(execute=@n1.then_4, Condition="Nearly Equal (Float)"(A="To Float (Boolean)"(InBool=@n5), B=1, ErrorTolerance=0.1))
   n7 = GetDeltaTransformFromRefPose(self=GetOwnerSkeletalMeshComponent(), BoneName=head, BaseName=neck_01)
   > Set head_fwdBwd_rot(head_fwdBwd_rot:double=BreakRotator(InRot=BreakTransform(InTransform=@n7).Rotation).Yaw) -> (Output_Get:double)
   Set head_fwdBwd_rot(execute=@n6.else, head_fwdBwd_rot=0)
   Event BlueprintInitializeAnimation()
   > Set body(body:Skeletal Mesh Component Object Reference=GetOwnerSkeletalMeshComponent()) -> (Output_Get:Skeletal Mesh Component Object Reference)
   Set SkelMesh'(SkelMesh':Skeletal Mesh Component Object Reference=GetOwnerSkeletalMeshComponent()) -> (Output_Get:Skeletal Mesh Component Object Reference)
   n8 = TryGetPawnOwner()
   > Is Valid(InputObject=@n8)
   > n9 = Cast To FortPlayerPawnAthena(Object=@n8)
   > Set PlayerPawn(PlayerPawn:Fort Player Pawn Athena Object Reference=@n9) -> (Output_Get:Fort Player Pawn Athena Object Reference)
   NOT Boolean(A=false)
   Gravity Override State Function DEPRECATED(CharacterPartAnimInstance=Self-Reference(), GameTimeInSeconds=GetGameTimeInSeconds(), GravityOverrideParams=Get Gravity Override Params())
   OR Boolean(A=NOT Boolean(A=Get bRigidBodyEnabled()), B=false)
   Make Gravity Override Params Struct!(
    Skydive_GoingUp_28_#1:Gravity Override Setting Struct Structure=Make Gravity Override Setting Struct!(
     WindFrequency_28_#2:Vector="50,20,40"
     WindAmplitude_31_#3:Vector="150,100,50"
     WindOffset_32_#4:Vector="0,0,0"
     GravityOverride_15_#5:Vector="0,0,-100"
     JointName_27_#6:Name=root
    ) -> (GravityOverrideSettingStruct:Gravity Override Setting Struct Structure) {is=title K2Node_MakeStruct}
    Skydive_FloatingDown_29_#7:Gravity Override Setting Struct Structure=Make Gravity Override Setting Struct(
     WindFrequency_28_#2="55,22,40"
     WindAmplitude_31_#3="300,-150,50"
     WindOffset_32_#4="0,0,0"
     GravityOverride_15_#5="0,-100,100"
     JointName_27_#6=root
    )
    Skydive_DivingDown_30_#8:Gravity Override Setting Struct Structure=Make Gravity Override Setting Struct(
     WindFrequency_28_#2="3,4,5"
     WindAmplitude_31_#3="150,200,60"
     WindOffset_32_#4="0,0,0"
     GravityOverride_15_#5="0,-300,600"
     JointName_27_#6=root
    )
    Skydive_Parachuting_31_#9:Gravity Override Setting Struct Structure=Make Gravity Override Setting Struct(
     WindFrequency_28_#2="1.5,7,8"
     WindAmplitude_31_#3="100,-200,200"
     WindOffset_32_#4="0,0,0"
     GravityOverride_15_#5="0,-300,-980"
     JointName_27_#6=root
    )
    OnGround_Standing_32_#10:Gravity Override Setting Struct Structure=(WindFrequency_28_#2=(X=0,Y=0,Z=0),WindAmplitude_31_#3=(X=0,Y=0,Z=0),WindOffset_32_#4=(X=0,Y=0,Z=0),GravityOverride_15_#5=(X=0,Y=0,Z=-980),JointName_27_#6="root")
    OnGround_Moving_33_#11:Gravity Override Setting Struct Structure=Make Gravity Override Setting Struct(
     WindFrequency_28_#2="0,0,0"
     WindAmplitude_31_#3="0,0,0"
     WindOffset_32_#4="0,0,0"
     GravityOverride_15_#5="0,0,-980"
     JointName_27_#6=root
    )
    EmoteOrMelee_34_#12:Gravity Override Setting Struct Structure=(WindFrequency_28_#2=(X=0,Y=0,Z=0),WindAmplitude_31_#3=(X=0,Y=0,Z=0),WindOffset_32_#4=(X=0,Y=0,Z=0),GravityOverride_15_#5=(X=0,Y=0,Z=-980),JointName_27_#6="root")
    Swimming_35_#13:Gravity Override Setting Struct Structure=(WindFrequency_28_#2=(X=0,Y=0,Z=0),WindAmplitude_31_#3=(X=0,Y=0,Z=0),WindOffset_32_#4=(X=0,Y=0,Z=0),GravityOverride_15_#5=(X=-800,Y=0,Z=0),JointName_27_#6="pelvis")
    Front_End_36_#14:Gravity Override Setting Struct Structure=(WindFrequency_28_#2=(X=0,Y=0,Z=0),WindAmplitude_31_#3=(X=0,Y=0,Z=0),WindOffset_32_#4=(X=0,Y=0,Z=0),GravityOverride_15_#5=(X=0,Y=0,Z=-980),JointName_27_#6="root")
    DBNO_40_#15:Gravity Override Setting Struct Structure=(WindFrequency_28_#2=(X=0,Y=0,Z=0),WindAmplitude_31_#3=(X=0,Y=0,Z=0),WindOffset_32_#4=(X=0,Y=0,Z=0),GravityOverride_15_#5=(X=-980,Y=400,Z=0),JointName_27_#6="pelvis")
    DBNO_Carried_41_#16:Gravity Override Setting Struct Structure=(WindFrequency_28_#2=(X=0,Y=0,Z=0),WindAmplitude_31_#3=(X=0,Y=0,Z=0),WindOffset_32_#4=(X=0,Y=0,Z=0),GravityOverride_15_#5=(X=-980,Y=500,Z=-500),JointName_27_#6="pelvis")
    OnGround_Falling_45_#17:Gravity Override Setting Struct Structure=(WindFrequency_28_#2=(X=0,Y=0,Z=0),WindAmplitude_31_#3=(X=0,Y=0,Z=0),WindOffset_32_#4=(X=0,Y=0,Z=0),GravityOverride_15_#5=(X=0,Y=0,Z=-980),JointName_27_#6="root")
    Ziplining_46_#18:Gravity Override Setting Struct Structure=(WindFrequency_28_#2=(X=0,Y=0,Z=0),WindAmplitude_31_#3=(X=0,Y=0,Z=0),WindOffset_32_#4=(X=0,Y=0,Z=0),GravityOverride_15_#5=(X=0,Y=0,Z=-980),JointName_27_#6="root")
    Skydive_Surfing_49_#19:Gravity Override Setting Struct Structure=(WindFrequency_28_#2=(X=1.5,Y=7,Z=8),WindAmplitude_31_#3=(X=200,Y=-500,Z=400),WindOffset_32_#4=(X=0,Y=0,Z=0),GravityOverride_15_#5=(X=500,Y=-300,Z=-980),JointName_27_#6="root")
    Riding_Standing_54_#20:Gravity Override Setting Struct Structure=(WindFrequency_28_#2=(X=0,Y=0,Z=0),WindAmplitude_31_#3=(X=0,Y=0,Z=0),WindOffset_32_#4=(X=0,Y=0,Z=0),GravityOverride_15_#5=(X=0,Y=-350,Z=-980),JointName_27_#6="root")
    Riding_Moving_55_#21:Gravity Override Setting Struct Structure=(WindFrequency_28_#2=(X=0,Y=0,Z=0),WindAmplitude_31_#3=(X=0,Y=0,Z=0),WindOffset_32_#4=(X=0,Y=0,Z=0),GravityOverride_15_#5=(X=0,Y=-450,Z=-450),JointName_27_#6="root")
    Riding_Moving_Targeting_58_#22:Gravity Override Setting Struct Structure=(WindFrequency_28_#2=(X=0,Y=0,Z=0),WindAmplitude_31_#3=(X=0,Y=0,Z=0),WindOffset_32_#4=(X=0,Y=0,Z=0),GravityOverride_15_#5=(X=0,Y=-200,Z=-980),JointName_27_#6="root")
    Tactical_Sprint_57_#23:Gravity Override Setting Struct Structure=Make Gravity Override Setting Struct(
     WindFrequency_28_#2="0,0,0"
     WindAmplitude_31_#3="0,0,0"
     WindOffset_32_#4="0,0,0"
     GravityOverride_15_#5="0,0,-980"
     JointName_27_#6=root
    )
    Rail_Grinding_61_#24:Gravity Override Setting Struct Structure=(WindFrequency_28_#2=(X=0,Y=0,Z=0),WindAmplitude_31_#3=(X=0,Y=0,Z=0),WindOffset_32_#4=(X=0,Y=0,Z=0),GravityOverride_15_#5=(X=0,Y=0,Z=-980),JointName_27_#6="root")
    DBNO_Tumbling_67_#25:Gravity Override Setting Struct Structure=(WindFrequency_28_#2=(X=0,Y=0,Z=0),WindAmplitude_31_#3=(X=0,Y=0,Z=0),WindOffset_32_#4=(X=0,Y=0,Z=0),GravityOverride_15_#5=(X=-2000,Y=0,Z=0),JointName_27_#6="pelvis")
    DBNO_RollFwd_68_#26:Gravity Override Setting Struct Structure=(WindFrequency_28_#2=(X=0,Y=0,Z=0),WindAmplitude_31_#3=(X=0,Y=0,Z=0),WindOffset_32_#4=(X=0,Y=0,Z=0),GravityOverride_15_#5=(X=-2000,Y=0,Z=0),JointName_27_#6="pelvis")
    DBNO_RollBwd_69_#27:Gravity Override Setting Struct Structure=(WindFrequency_28_#2=(X=0,Y=0,Z=0),WindAmplitude_31_#3=(X=0,Y=0,Z=0),WindOffset_32_#4=(X=0,Y=0,Z=0),GravityOverride_15_#5=(X=-4000,Y=0,Z=0),JointName_27_#6="pelvis")
    DBNO_Impulse_70_#28:Gravity Override Setting Struct Structure=(WindFrequency_28_#2=(X=0,Y=0,Z=0),WindAmplitude_31_#3=(X=0,Y=0,Z=0),WindOffset_32_#4=(X=0,Y=0,Z=0),GravityOverride_15_#5=(X=-2000,Y=0,Z=0),JointName_27_#6="pelvis")
   ) -> (GravityOverrideParamsStruct:Gravity Override Params Struct Structure) {is=title K2Node_MakeStruct}
   n10 = GetAlphaFromReferenceTransformDelta(
    SkeletalMeshComponent=Get Custom Base Mesh() -> (Custom Base Mesh:Skeletal Mesh Component Object Reference)
    BoneName=spine_05
    BaseName=head
    DeltaComponentType=EulerX
    InputRange=(X=-60,Y=60)
    OutputRange=(X=-1,Y=1)
    TwistAxis="0,0,1"
   )
   Set Hair Side Rotation(Hair Side Rotation:Rotator=MakeRotator(Pitch=float * float(A=@n10, B=-120) {is=title K2Node_PromotableOperator})) -> (Output_Get:Rotator)
   Set Hair Up(Hair Up:double=GetAlphaFromReferenceTransformDelta(BoneName=spine_05, BaseName=head, InputRange=(X=0,Y=70), TwistAxis="0,0,1")) -> (Output_Get:double)
  ]}
]
variables [
 body {type=SkeletalMeshComponent}
 ComponentLinearAccScale {type=Vector, default=<0.35 0.35 0.35>}
 ComponentLinearVelScale {type=Vector, default=<0.35 0.35 0.35>}
 Custom Base Mesh {type=SkeletalMeshComponent}
 Gravity Override Params {type=GravityOverrideParamsStruct, default {skydive_Parachuting {windAmplitude=<50 -500 50>, windFrequency=<1.5 7 3>}}}
 GravityOverride {type=Vector}
 hair Rot {type=number}
 Hair Side Rotation {type=Rotator}
 Hair Up {type=number}
 head_fwdBwd_rot {type=number}
 is Wearing Backbling {type=boolean}
 is Wearing Cape {type=boolean}
 IsNotParashuting {type=boolean}
 NewVar {type=SkeletalMeshComponent}
 PlayerPawn {type=FortPlayerPawnAthena}
 SkelMesh' {type=SkeletalMeshComponent}
 SkydiveFloatPoseFixAlpha {type=number}
]
