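# Control Rig for the BrightDisk female character's face accessory, with correctives for the hair tail and headphones bones.
# Headphone offset is weighted by a delta-reference height; head spherical pose readers, one Kalman-smoothed, weight hair-tail adjustments when RigidBody Enabled is false.
# A pelvis-height check gates a smoothed left-upper-arm reader that bends the tail base; the right-arm reader drives nothing.
# Begin Execution connects to none of these chains.
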ControlRigBlueprint /BRCosmetics/Characters/Player/Female/Medium/Bodies/F_MED_BrightDisk/Meshes/Parts/F_MED_BrightDisk_FaceAcc_CtrlRig
graphs [
 RigVMFunctionLibrary {links [], nodes []}
 RigVMModel {nodes [
   Get Delta Ref = RigVMFunctionReferenceNode(ExecuteContext=(), Item=(Type=Bone,Name="root"))
   > &RigUnit_ModifyTransforms(
    ItemToModify=((Item=(Type=Bone,Name="Headphones_fix"),Transform=(Rotation=(X=0,Y=0,Z=0,W=1),Translation=(X=0,Y=-0.5,Z=0.4),Scale3D=(X=1,Y=1,Z=1))))
    WeightMinimum=0
    Weight=&RigVMFunction_MathDoubleRemap(SourceMaximum=0, SourceMinimum=-1, TargetMaximum=1, TargetMinimum=0, Value=@"Get Delta Ref".Result.Location.Z).Result
   )
   n1 = &RigVMFunction_ControlFlowBranch(Condition=False, ExecuteContext=())
   &RigUnit_ModifyTransforms(
    ItemToModify="((Item=(Type=Bone,Name=\"base_hairTail\"),Transform=(Rotation=(X=0,Y=0,Z=-0.284,W=0.9588),Translation=(X=0,Y=0,Z=0),Scale3D=(X=1,Y=1,Z=1))),(Item=(Type=Bone,Name=\"hairTail_01\"),Transform=(Rotation=(X=0,Y=0,Z=-0.0872,W=0.9962),Translation=(X=0,Y=0,Z=0),Scale3 ...(truncated)"
    Weight=1
    WeightMinimum=0
    ExecutePin=@n1.True
   )
   n2 = &RigVMFunction_ControlFlowBranch(ExecuteContext=(), Condition=RigVMVariableNode() {variable=RigidBody Enabled}.Value)
   n3 = &RigUnit_SphericalPoseReader(
    ActiveRegionScaleFactors=(PositiveWidth=3,NegativeWidth=0.7,PositiveHeight=3.5,NegativeHeight=2)
    ActiveRegionSize=0.1
    Debug=(bDrawDebug=true,bDraw2D=False,bDrawLocalAxes=False,DebugScale=25,DebugSegments=20,DebugThickness=0.25)
    DriverAxis=(X=0,Y=1,Z=0)
    DriverItem=(Type=Bone,Name="head")
    FalloffRegionScaleFactors=(PositiveWidth=1,NegativeWidth=1,PositiveHeight=3,NegativeHeight=0.5)
    FalloffSize=0.4
    FlipHeightScaling=False
    FlipWidthScaling=False
    OptionalParentItem=(Type=Bone,Name="spine_05")
    RotationOffset=(X=90,Y=0,Z=-60)
    ExecutePin=@n2.False
   )
   KalmanFilter = &RigVMFunction_KalmanFloat(BufferSize=16, Value=@n3.OutputParam) {template="KalmanFilter::Execute(in Value,in BufferSize,out Result)"}
   > RigUnit_SetRelativeTransformForItem = /Script/ControlRig.RigUnit_SetRelativeTransformForItem(
    Child=(Type=Bone,Name="hairTail_02")
    Parent=(Type=Bone,Name="hairTail_02")
    Value=(Rotation=(X=0,Y=0,Z=0,W=1),Translation=(X=0,Y=0,Z=0),Scale3D=(X=1,Y=1,Z=1))
    bParentInitial=False
    bPropagateToChildren=True
    Weight=@KalmanFilter.Result
   ) {
    is=struct RigVMUnitNode
    template="Set Relative Transform::Execute(in Child,in Parent,in bParentInitial,in Value,in Weight,in bPropagateToChildren)"
   }
   > n4 = &RigUnit_SphericalPoseReader(
    ActiveRegionScaleFactors=(PositiveWidth=3,NegativeWidth=0.7,PositiveHeight=3.5,NegativeHeight=2)
    ActiveRegionSize=0.1
    Debug=(bDrawDebug=true,bDraw2D=False,bDrawLocalAxes=False,DebugScale=25,DebugSegments=20,DebugThickness=0.25)
    DriverAxis=(X=0,Y=1,Z=0)
    DriverItem=(Type=Bone,Name="head")
    FalloffRegionScaleFactors=(PositiveWidth=1,NegativeWidth=0.5,PositiveHeight=3,NegativeHeight=0.5)
    FalloffSize=0.4
    FlipHeightScaling=False
    FlipWidthScaling=False
    OptionalParentItem=(Type=Bone,Name="spine_05")
    RotationOffset=(X=90,Y=0,Z=-45)
   )
   > &RigUnit_SetRelativeTransformForItem(
    Child=(Type=Bone,Name="hairTail_01")
    Parent=(Type=Bone,Name="hairTail_01")
    Value=(Rotation=(X=0,Y=0,Z=0,W=1),Translation=(X=0,Y=0,Z=0),Scale3D=(X=1,Y=1,Z=1))
    bParentInitial=False
    bPropagateToChildren=True
    Weight=@n4.OutputParam
   ) {template="Set Relative Transform::Execute(in Child,in Parent,in bParentInitial,in Value,in Weight,in bPropagateToChildren)"}
   > &RigUnit_SetRelativeTransformForItem(
    Child=(Type=Bone,Name="hairTail_02")
    Parent=(Type=Bone,Name="hairTail_02")
    Value=(Rotation=(X=0,Y=0,Z=0,W=1),Translation=(X=0,Y=0,Z=0),Scale3D=(X=1,Y=1,Z=1))
    bParentInitial=False
    bPropagateToChildren=True
    Weight=@n4.OutputParam
   ) {template="Set Relative Transform::Execute(in Child,in Parent,in bParentInitial,in Value,in Weight,in bPropagateToChildren)"}
   > &RigUnit_SetRelativeTransformForItem(
    Child=(Type=Bone,Name="hairTail_03")
    Parent=(Type=Bone,Name="hairTail_03")
    Value=(Rotation=(X=0,Y=0,Z=0,W=1),Translation=(X=0,Y=0,Z=0),Scale3D=(X=1,Y=1,Z=1))
    bParentInitial=False
    bPropagateToChildren=True
    Weight=@n4.OutputParam
   ) {template="Set Relative Transform::Execute(in Child,in Parent,in bParentInitial,in Value,in Weight,in bPropagateToChildren)"}
   RigVMFunction_Sequence = &RigVMFunction_Sequence(ExecuteContext=())
   n5 = &RigUnit_GetRelativeTransformForItem(Child=(Type=Bone,Name="pelvis"), Parent=(Type=Bone,Name="root"), bChildInitial=False, bParentInitial=true)
   n6 = &RigUnit_GetRelativeTransformForItem(Child=(Type=Bone,Name="pelvis"), Parent=(Type=Bone,Name="root"), bChildInitial=False, bParentInitial=true)
   DISPATCH_RigVMDispatch_CoreEquals = RigVMDispatchNode(
    B=0
    A=&RigVMFunction_MathDoubleMul(
     B=&RigVMFunction_MathDoubleRemap(SourceMaximum=40, SourceMinimum=55, TargetMaximum=1, TargetMinimum=0, Value=@n5.RelativeTransform.Translation.Z).Result
     A=&RigVMFunction_MathDoubleRemap(SourceMaximum=75, SourceMinimum=80, TargetMaximum=1, TargetMinimum=0, Value=@n6.RelativeTransform.Translation.Z).Result
    ).Result
   ) {template="DISPATCH_RigVMDispatch_CoreEquals(in A,in B,out Result)"}
   RigVMFunction_ControlFlowBranch = &RigVMFunction_ControlFlowBranch(Condition=@DISPATCH_RigVMDispatch_CoreEquals.Result, ExecuteContext=@RigVMFunction_Sequence.A)
   DISPATCH_RigVMDispatch_Print = RigVMDispatchNode(
    Enabled=True
    Prefix="              "
    ScreenColor=(R=1,G=1,B=1,A=1)
    ScreenDuration=0.05
    Value=@DISPATCH_RigVMDispatch_CoreEquals.Result
    ExecuteContext=@RigVMFunction_ControlFlowBranch.False
   ) {template="DISPATCH_RigVMDispatch_Print(in Prefix,in Value,in Enabled,in ScreenDuration,in ScreenColor)"}
   n7 = &RigUnit_SphericalPoseReader(
    ActiveRegionScaleFactors=(PositiveWidth=1,NegativeWidth=1,PositiveHeight=1,NegativeHeight=1)
    ActiveRegionSize=0.2
    Debug=(bDrawDebug=true,bDraw2D=False,bDrawLocalAxes=False,DebugScale=25,DebugSegments=20,DebugThickness=0.25)
    DriverAxis=(X=1,Y=0,Z=0)
    DriverItem=(Type=Bone,Name="upperarm_l")
    FalloffRegionScaleFactors=(PositiveWidth=1,NegativeWidth=1,PositiveHeight=1,NegativeHeight=1)
    FalloffSize=0.3
    FlipHeightScaling=False
    FlipWidthScaling=false
    OptionalParentItem=(Type=Bone,Name="None")
    RotationOffset=(X=0,Y=-50,Z=0)
    ExecutePin=@RigVMFunction_ControlFlowBranch.True
   )
   > &RigUnit_ModifyTransforms(
    ItemToModify=((Item=(Type=Bone,Name="base_hairTail"),Transform=(Rotation=(X=0,Y=0,Z=-0.4226,W=0.9063),Translation=(X=1,Y=0,Z=0),Scale3D=(X=1,Y=1,Z=1))))
    WeightMinimum=0
    Weight=&RigVMFunction_KalmanFloat(BufferSize=8, Value=@n7.OutputParam) {template="KalmanFilter::Execute(in Value,in BufferSize,out Result)"}.Result
   )
   &RigUnit_SphericalPoseReader(
    ActiveRegionScaleFactors=(PositiveWidth=1,NegativeWidth=1,PositiveHeight=1,NegativeHeight=1)
    ActiveRegionSize=0.1
    Debug=(bDrawDebug=true,DebugScale=25,DebugSegments=20,DebugThickness=0.25)
    DriverAxis=(X=-1,Y=0,Z=0)
    DriverItem=(Type=Bone,Name="upperarm_r")
    FalloffRegionScaleFactors=(PositiveWidth=1,NegativeWidth=1,PositiveHeight=1,NegativeHeight=1)
    FalloffSize=0.5
    FlipHeightScaling=true
    FlipWidthScaling=false
    OptionalParentItem=(Type=Bone,Name="None")
    RotationOffset=(X=180,Y=0,Z=180)
    ExecutePin=@RigVMFunction_Sequence.B
   )
   RigUnit_BeginExecution = &RigUnit_BeginExecution()
   RigVMVariableNode() {variable=RigidBody Enabled}
  ]}
]
hierarchy [
 root {loc=<0 0 0>, rot=<0 0 0 1>}
 pelvis {parent=root, loc=<0 -0.5301 98.0571>, rot=<0.0127 0.707 -0.0127 -0.707>}
 spine_01 {parent=pelvis, loc=<6.1208 -0.9328 0>, rot=<0 0 0.034 -0.9994>}
 spine_02 {parent=spine_01, loc=<6.5996 -0.2095 0>, rot=<0 0 0.0238 -0.9997>}
 spine_03 {parent=spine_02, loc=<6.4914 -0.1237 0>, rot=<0 0 0.0149 -0.9999>}
 spine_04 {parent=spine_03, loc=<6.9598 0.0794 0>, rot=<0 0 -0.0132 -0.9999>}
 spine_05 {parent=spine_04, loc=<6.9987 0 0>, rot=<0 0 -0.0015 -1>}
 neck_01 {parent=spine_05, loc=<11.3138 0.1529 0>, rot=<0 0 -0.2003 -0.9797>}
 neck_02 {parent=neck_01, loc=<4.5627 0 0>, rot=<0 0 0.0504 -0.9987>}
 head {parent=neck_02, loc=<5.3914 0 0>, rot=<0 0 0.115 -0.9934>}
 base_hairTail {parent=head, loc=<14.1093 -4.4509 -0.2005>, rot=<0.9624 -0.2684 0.0222 -0.0369>}
 hairTail_01 {parent=base_hairTail, loc=<0 0 0>, rot=<0 0 0 1>}
 hairTail_02 {parent=hairTail_01, loc=<6.2094 0.6794 0.0864>, rot=<0 0 -0.7871 -0.6169>}
 hairTail_03 {parent=hairTail_02, loc=<10.0415 3.7432 -0.8761>, rot=<0 0 -0.3539 -0.9353>}
 hairTail_04 {parent=hairTail_03, loc=<8.8208 1.0251 -0.5285>, rot=<0 0 -0.0067 -1>}
 hairTail_05 {parent=hairTail_04, loc=<10.0846 -1.8538 -0.7656>, rot=<0 0 0 1>}
 base_hairSide_l {parent=head, loc=<2.2548 5.2517 -6.4305>, rot=<0.0368 0.0079 0.9772 -0.2091>}
 hairSide_l_00 {parent=base_hairSide_l, loc=<0 0 0>, rot=<0 0 0 1>}
 hairSide_l_01 {parent=hairSide_l_00, loc=<2.9362 0 0>, rot=<0 0 0 1>}
 base_hairSide_r {parent=head, loc=<3.1598 5.6577 6.7113>, rot=<0.211 -0.9746 0.071 0.0232>}
 hairSide_r_00 {parent=base_hairSide_r, loc=<0 0 0>, rot=<0 0 0 1>}
 hairSide_r_01 {parent=hairSide_r_00, loc=<2.9362 0 0>, rot=<0 0 0 1>}
 base_hair_f_l {parent=head, loc=<10.0602 10.3017 -1.9546>, rot=<-0.0482 -0.9487 -0.0192 -0.3119>}
 hair_f_l_0 {parent=base_hair_f_l, loc=<0 0 0>, rot=<0 0 0 1>}
 hair_f_l_1 {parent=hair_f_l_0, loc=<5.0843 0 0>, rot=<0 0 0 1>}
 base_hair_f_m {parent=head, loc=<8.9122 10.7878 1.2069>, rot=<-0.0444 -0.987 -0.0268 -0.1523>}
 hair_f_m_0 {parent=base_hair_f_m, loc=<0 0 0>, rot=<0 0 0 1>}
 hair_f_m_1 {parent=hair_f_m_0, loc=<3.9194 0 0>, rot=<0 0 0 1>}
 base_hair_r {parent=head, loc=<9.275 10.1851 3.3663>, rot=<0.0476 -0.9963 -0.0513 0.0498>}
 hair_f_r_0 {parent=base_hair_r, loc=<0 0 0>, rot=<0 0 0 1>}
 hair_f_r_1 {parent=hair_f_r_0, loc=<3.9194 0 0>, rot=<0 0 0 1>}
 Headphones_fix {parent=head, loc=<0 0 0>, rot=<0.7007 0.0028 0.7134 -0.003>}
 clavicle_l {parent=spine_05, loc=<5.4831 -0.1053 -3.2946>, rot=<0.0465 -0.7289 -0.009 -0.6829>}
 upperarm_l {parent=clavicle_l, loc=<9.8658 -1.2136 -0.4105>, rot=<0.0164 -0.3245 -0.0269 -0.9454>}
 lowerarm_l {parent=upperarm_l, loc=<26.465 -0.8345 0.259>, rot=<0.0036 0.0161 -0.2298 -0.9731>}
 clavicle_r {parent=spine_05, loc=<5.4831 -0.1053 3.2946>, rot=<0.7289 0.0465 -0.6829 0.009>}
 upperarm_r {parent=clavicle_r, loc=<-9.8658 1.2136 0.4105>, rot=<0.0164 -0.3245 -0.0269 -0.9454>}
 lowerarm_r {parent=upperarm_r, loc=<-26.4651 0.8345 -0.259>, rot=<0.0036 0.0161 -0.2298 -0.9731>}
]
variables [
 Is emote {type=bool, default=False}
 RigidBody Enabled {type=bool, default=False}
]
